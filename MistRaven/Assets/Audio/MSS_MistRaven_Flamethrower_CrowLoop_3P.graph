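# Looping mono MetaSound for the third-person crow layer of the MistRaven flamethrower, using cone attenuation.
# On play, it shuffles the Sounds array (one crow loop by default) and triggers an ADSR attack that starts the wave player with randomized pitch.
# Randomized volume, the envelope and an affiliation mix gain scale the output in turn.
# Only the envelope's attack trigger is wired, so its release never fires within the graph.
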MetaSoundSource /MistRaven/Assets/Audio/MSS_MistRaven_Flamethrower_CrowLoop_3P
inputs [
 Character.Affiliation {type=Enum:FortAudioAffiliation}
 Character.Health {type=Float}
 Character.UniqueID {type=Int32}
 Random.Pitch.Base {type=Float}
 Random.Pitch.Spread {type=Float}
 Random.Volume.Base {type=Float}
 Random.Volume.Spread {type=Float}
 Sounds {type=WaveAsset:Array}
 UE.Source.OnPlay {type=Trigger}
]
interfaces [Fort.Character:1.0, UE.OutputFormat.Mono:1.0, UE.Source:1.1]
outputs ["UE.OutputFormat.Mono.Audio:0" {type=Audio}]
pages [Default {
  input_defaults {
   Character.Affiliation=0
   Character.Health=0
   Character.UniqueID=0
   Random.Pitch.Base=0
   Random.Pitch.Spread=1
   Random.Volume.Base=2
   Random.Volume.Spread=0
   UE.Source.OnPlay=null
   Sounds [&/MistRaven_Flamethrower_CrowLoop_01]
  }
  nodes [
   n1 = Array.Shuffle:WaveAsset:Array(In Array=^Sounds, Next=^UE.Source.OnPlay) {also {defaults {Enable Shared State="<MetaSoundBuilderResult.SUCCEEDED: 0>"}}}
   ADSREnvelopeADSREnvelope_0 = ADSR Envelope.ADSR Envelope:Audio(Trigger Attack=@n1.On Next) {also {defaults {Attack Time=0.2, Release Time=4, Sustain Level=1}}}
   n2 = Max.Max:Float(A=UE.RandomFloat(
     Max=UE.Add:Float(PrimaryOperand=^Random.Volume.Base, AdditionalOperands=^Random.Volume.Spread).Out
     Next=UE.Reroute(Value=^UE.Source.OnPlay).Value
     Min=UE.Subtract:Float(PrimaryOperand=^Random.Volume.Base, AdditionalOperands=^Random.Volume.Spread).Out
    ).Value)
   WavePlayer_0 = UE.Wave Player:Mono(
    Play=@ADSREnvelopeADSREnvelope_0.On Attack Triggered
    Wave Asset=@n1.Value
    Pitch Shift=UE.RandomFloat(
     Max=UE.Add:Float(PrimaryOperand=^Random.Pitch.Base, AdditionalOperands=^Random.Pitch.Spread).Out
     Next=UE.Reroute(Value=^UE.Source.OnPlay).Value
     Min=UE.Subtract:Float(PrimaryOperand=^Random.Pitch.Base, AdditionalOperands=^Random.Pitch.Spread).Out
    ).Value
   ) {also {defaults {Loop="<MetaSoundBuilderResult.SUCCEEDED: 0>"}}}
   Multiply_1 = UE.Multiply:Audio by Float(
    AdditionalOperands=&MSP_AffiliationMixFloat(Affiliation=^Character.Affiliation) {also {defaults {Non-Enemy=0.5}}}.Out
    PrimaryOperand=UE.Multiply:Audio(
     AdditionalOperands=@ADSREnvelopeADSREnvelope_0.Out Envelope
     PrimaryOperand=UE.Multiply:Audio by Float(AdditionalOperands=@n2.Value, PrimaryOperand=@WavePlayer_0.Out Mono).Out
    ).Out
   )
  ]
  links [{dst="@UE.OutputFormat.Mono.Audio:0", src=Multiply_1.Out}]
 }]
properties {
 assetImportData=&self:AssetImportData
 attenuationSettings=&/SA_MistRaven_Flamethrower_Loops_Cone
 bDebug=true
 bLooping=true
 duration=10000
 maxDistance=6300
 sampleRate=48000
 soundClassObject=&Fort_WeaponMelee
 volume=1.5
}
