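# Looping stereo MetaSound for the MistRaven flamethrower's first-person energy loop, layering a wave player with a synthesized low hum.
# The shuffled wave gets randomized volume and pitch, plus a pitch offset that rises as the smoothed AmmoInMagazine value falls.
# The hum is a sine amplitude-modulated by an LFO; both frequencies climb over five seconds after play.
# Volume and a fade-in envelope scale only the hum before both layers are mixed.
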MetaSoundSource /MistRaven/Assets/Audio/MSS_MistRaven_Flamethrower_EnergyLoop_1P
inputs [
 AmmoInMagazine {type=Float}
 Fade In Time {type=Time}
 Random.Pitch.Base {type=Float}
 Random.Pitch.Spread {type=Float}
 Random.Volume.Base {type=Float}
 Random.Volume.Spread {type=Float}
 Sounds {type=WaveAsset:Array}
 UE.Source.OnPlay {type=Trigger}
 Volume {type=Float}
]
interfaces [UE.OutputFormat.Stereo:1.0, UE.Source:1.1]
outputs [
 "UE.OutputFormat.Stereo.Audio:0" {type=Audio}
 "UE.OutputFormat.Stereo.Audio:1" {type=Audio}
]
pages [Default {
  input_defaults {
   AmmoInMagazine=0
   Fade In Time=0.5
   Random.Pitch.Base=0
   Random.Pitch.Spread=2
   Random.Volume.Base=0.9
   Random.Volume.Spread=0
   UE.Source.OnPlay=null
   Volume=0.5
   Sounds [&/MistRaven_Flamethrower_EnergyLoop_01]
  }
  nodes [
   n1 = Max.Max:Float(A=UE.RandomFloat(
     Max=UE.Add:Float(PrimaryOperand=^Random.Volume.Base, AdditionalOperands=^Random.Volume.Spread).Out
     Next=UE.Reroute(Value=^UE.Source.OnPlay).Value
     Min=UE.Subtract:Float(PrimaryOperand=^Random.Volume.Base, AdditionalOperands=^Random.Volume.Spread).Out
    ).Value) {also {defaults {B=0}}}
   n2 = UE.Subtract:Float(AdditionalOperands=UE.InterpTo:Audio(Target=^AmmoInMagazine).Value) {also {defaults {PrimaryOperand=1}}}
   n3 = UE.Add:Float(PrimaryOperand=^Random.Pitch.Base, AdditionalOperands=^Random.Pitch.Spread)
   n4 = UE.Reroute(Value=^UE.Source.OnPlay)
   n5 = UE.Subtract:Float(PrimaryOperand=^Random.Pitch.Base, AdditionalOperands=^Random.Pitch.Spread)
   n6 = Array.Shuffle:WaveAsset:Array(In Array=^Sounds, Next=^UE.Source.OnPlay) {also {defaults {Enable Shared State="<MetaSoundBuilderResult.SUCCEEDED: 0>"}}}
   WavePlayer_0 = UE.Wave Player:Stereo(
    Pitch Shift=UE.Add:Float(
     PrimaryOperand=UE.Multiply:Float(PrimaryOperand=@n2.Out) {also {defaults {AdditionalOperands=12}}}.Out
     AdditionalOperands=UE.RandomFloat(Max=@n3.Out, Next=@n4.Value, Min=@n5.Out).Value
    ).Out
    Play=@n6.On Next
    Wave Asset=@n6.Value
   ) {also {defaults {Loop="<MetaSoundBuilderResult.SUCCEEDED: 0>"}}}
   ADSREnvelopeADSREnvelope_0 = ADSR Envelope.ADSR Envelope:Float(Attack Time=^Fade In Time, Trigger Attack=UE.Reroute(Value=^UE.Source.OnPlay).Value) {also {defaults {Attack Curve=1.25, Sustain Level=1}}}
   ADSREnvelopeADSREnvelope_1 = ADSR Envelope.ADSR Envelope:Float(Trigger Attack=^UE.Source.OnPlay) {also {defaults {Attack Curve=0.6, Attack Time=5, Sustain Level=1}}}
   n7 = UE.LFO:Audio(Frequency=UE.Add:Float(AdditionalOperands=MapRange.MapRange:Float(In=@ADSREnvelopeADSREnvelope_1.Out Envelope) {also {defaults {Out Range B=4}}}.Out Value) {also {defaults {PrimaryOperand=8}}}.Out) {also {defaults {Min Value=0, Shape=2}}}
   n8 = UE.Sine:Audio(Frequency=UE.Add:Float(PrimaryOperand=HarmonixNodes.Ramper(Snap To Start=^UE.Source.OnPlay) {also {defaults {Ramp Time=5, Start=0, Target=10}, version [0,1]}}.Value) {also {defaults {AdditionalOperands=56}}}.Out)
   n9 = UE.Multiply:Audio by Float(
    AdditionalOperands=@ADSREnvelopeADSREnvelope_0.Out Envelope
    PrimaryOperand=UE.Multiply:Audio by Float(AdditionalOperands=^Volume, PrimaryOperand=UE.Multiply:Audio by Float(AdditionalOperands=@n7.Out, PrimaryOperand=@n8.Audio).Out).Out
   )
   AudioMixerAudioMixerStereo2_0 = "AudioMixer.Audio Mixer (Stereo, 2)"(
    In 1 R=UE.Multiply:Audio by Float(AdditionalOperands=@n1.Value, PrimaryOperand=@WavePlayer_0.Out Right).Out
    In 1 L=UE.Multiply:Audio by Float(AdditionalOperands=@n1.Value, PrimaryOperand=@WavePlayer_0.Out Left).Out
    In 0 L=@n9.Out
    In 0 R=@n9.Out
   ) {also {defaults {Gain 0=0.75}}}
   VariableMutator.Trigger(Value=@n6.On Shuffle, Variable=InitVariable.Trigger().Variable)
  ]
  links [
   {dst="@UE.OutputFormat.Stereo.Audio:0", src=AudioMixerAudioMixerStereo2_0.Out L}
   {dst="@UE.OutputFormat.Stereo.Audio:1", src=AudioMixerAudioMixerStereo2_0.Out R}
  ]
 }]
properties {
 assetImportData=&self:AssetImportData
 bLooping=true
 duration=10000
 numChannels=2
 outputFormat=Stereo
 sampleRate=48000
 soundClassObject=&Fort_WeaponMelee
}
