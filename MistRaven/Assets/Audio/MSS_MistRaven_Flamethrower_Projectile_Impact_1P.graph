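# Mono one-shot MetaSound Source for the first-person impact sound of the MistRaven flamethrower projectile.
# On play, a short random delay triggers a no-repeat random pick from the impact wave array into a wave player with randomized pitch.
# The output is scaled by a randomized volume passed through a Max node, then by an affiliation mix for local, enemy and non-enemy characters.
# Dedicated attenuation and concurrency assets govern playback.
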MetaSoundSource /MistRaven/Assets/Audio/MSS_MistRaven_Flamethrower_Projectile_Impact_1P
inputs [
 Affiliation Volume - Enemy {type=Float}
 Affiliation Volume - Local {type=Float}
 Affiliation Volume - Non-Enemy {type=Float}
 Character.Affiliation {type=Enum:FortAudioAffiliation}
 Character.Health {type=Float}
 Character.UniqueID {type=Int32}
 Random.Pitch.Base {type=Float}
 Random.Pitch.Spread {type=Float}
 Random.Volume.Base {type=Float}
 Random.Volume.Spread {type=Float}
 Sounds {type=WaveAsset:Array}
 UE.Source.OnPlay {type=Trigger}
]
interfaces [Fort.Character:1.0, UE.OutputFormat.Mono:1.0, UE.Source.OneShot:1.0, UE.Source:1.1]
outputs [
 "UE.OutputFormat.Mono.Audio:0" {type=Audio}
 UE.Source.OneShot.OnFinished {type=Trigger}
]
pages [Default {
  input_defaults {
   Affiliation Volume - Enemy=1
   Affiliation Volume - Local=1
   Affiliation Volume - Non-Enemy=1
   Character.Affiliation=0
   Character.Health=0
   Character.UniqueID=0
   Random.Pitch.Base=0
   Random.Pitch.Spread=2
   Random.Volume.Base=1.5
   Random.Volume.Spread=0.1
   UE.Source.OnPlay=null
   Sounds [
    &/MistRaven_Flamethrower_Projectile_Impact_01
    &/MistRaven_Flamethrower_Projectile_Impact_02
    &/MistRaven_Flamethrower_Projectile_Impact_03
    &/MistRaven_Flamethrower_Projectile_Impact_04
    &/MistRaven_Flamethrower_Projectile_Impact_05
    &/MistRaven_Flamethrower_Projectile_Impact_06
   ]
  }
  nodes [
   n1 = UE.RandomFloat(
    Max=UE.Add:Float(PrimaryOperand=^Random.Volume.Base, AdditionalOperands=^Random.Volume.Spread).Out
    Next=UE.Reroute(Value=^UE.Source.OnPlay).Value
    Min=UE.Subtract:Float(PrimaryOperand=^Random.Volume.Base, AdditionalOperands=^Random.Volume.Spread).Out
   )
   TriggerDelay_0 = UE.Trigger Delay(
    In=^UE.Source.OnPlay
    Delay Time=UE.ConversionFloatToTime(In=/FortMetasound/Patches/RandomFloatValueByTrigger(TriggerIn=^UE.Source.OnPlay) {also {defaults {FloatMax=0.1, FloatMin=0}}}.Output).Out
   )
   ArrayRandomGet_0 = Array.Random Get:WaveAsset:Array(In Array=^Sounds, Next=@TriggerDelay_0.Out) {also {defaults {Enable Shared State="<MetaSoundBuilderResult.SUCCEEDED: 0>", No Repeats=7, Weights []}}}
   WavePlayer_0 = UE.Wave Player:Mono(
    Play=@ArrayRandomGet_0.On Next
    Wave Asset=@ArrayRandomGet_0.Value
    Pitch Shift=UE.RandomFloat(
     Max=UE.Add:Float(PrimaryOperand=^Random.Pitch.Base, AdditionalOperands=^Random.Pitch.Spread).Out
     Next=UE.Reroute(Value=^UE.Source.OnPlay).Value
     Min=UE.Subtract:Float(PrimaryOperand=^Random.Pitch.Base, AdditionalOperands=^Random.Pitch.Spread).Out
    ).Value
   )
   Multiply_0 = UE.Multiply:Audio by Float(
    AdditionalOperands=&MSP_AffiliationMixFloat(
     Enemy=^Affiliation Volume - Enemy
     Local=^Affiliation Volume - Local
     Non-Enemy=^Affiliation Volume - Non-Enemy
     Affiliation=^Character.Affiliation
    ).Out
    PrimaryOperand=UE.Multiply:Audio by Float(AdditionalOperands=Max.Max:Float(A=@n1.Value).Value, PrimaryOperand=@WavePlayer_0.Out Mono).Out
   )
  ]
  links [{dst="@UE.OutputFormat.Mono.Audio:0", src=Multiply_0.Out}, {dst="@UE.Source.OneShot.OnFinished", src=WavePlayer_0.On Finished}]
 }]
properties {
 assetImportData=&self:AssetImportData
 attenuationSettings=&/SA_MistRaven_Flamethrower_Projectile_Impact
 maxDistance=8300
 sampleRate=48000
 soundClassObject=&Fort_Impacts
 concurrencyOverrides {bLimitToOwner=true, maxCount=4, resolutionRule=StopFarthestThenPreventNew}
 concurrencySet [&/Conc_Mistraven_ProjectileImpact_1P]
}
