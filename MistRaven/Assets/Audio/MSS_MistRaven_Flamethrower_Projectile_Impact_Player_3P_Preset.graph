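# MetaSound Source preset for the third-person player impact sound of MistRaven flamethrower projectiles, played as a mono one-shot.
# It inherits its graph from the 1P impact MetaSound, feeding it six impact waves plus random pitch and volume ranges.
# Fort.Character affiliation inputs accompany per-affiliation volumes, with non-enemy set lower than local and enemy.
# Its own attenuation, impacts sound class and 3P concurrency set govern playback.
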MetaSoundSource /MistRaven/Assets/Audio/MSS_MistRaven_Flamethrower_Projectile_Impact_Player_3P_Preset
preset_parent=&/MSS_MistRaven_Flamethrower_Projectile_Impact_1P
inputs [
 Affiliation Volume - Enemy {type=Float}
 Affiliation Volume - Local {type=Float}
 Affiliation Volume - Non-Enemy {type=Float}
 Character.Affiliation {type=Enum:FortAudioAffiliation}
 Character.Health {type=Float}
 Character.UniqueID {type=Int32}
 Random.Pitch.Base {type=Float}
 Random.Pitch.Spread {type=Float}
 Random.Volume.Base {type=Float}
 Random.Volume.Spread {type=Float}
 Sounds {type=WaveAsset:Array}
 UE.Source.OnPlay {type=Trigger}
]
interfaces [Fort.Character:1.0, UE.OutputFormat.Mono:1.0, UE.Source.OneShot:1.0, UE.Source:1.1]
outputs [
 "UE.OutputFormat.Mono.Audio:0" {type=Audio}
 UE.Source.OneShot.OnFinished {type=Trigger}
]
pages [Default {input_defaults {
   Affiliation Volume - Enemy=1
   Affiliation Volume - Local=1
   Affiliation Volume - Non-Enemy=0.4
   Character.Affiliation=0
   Character.Health=0
   Character.UniqueID=0
   Random.Pitch.Base=0
   Random.Pitch.Spread=2
   Random.Volume.Base=2.8
   Random.Volume.Spread=0.1
   UE.Source.OnPlay=null
   Sounds [
    &/MistRaven_Flamethrower_Projectile_Impact_01
    &/MistRaven_Flamethrower_Projectile_Impact_02
    &/MistRaven_Flamethrower_Projectile_Impact_03
    &/MistRaven_Flamethrower_Projectile_Impact_04
    &/MistRaven_Flamethrower_Projectile_Impact_05
    &/MistRaven_Flamethrower_Projectile_Impact_06
   ]
  }}]
properties {
 assetImportData=&self:AssetImportData
 attenuationSettings=&/SA_MistRaven_Flamethrower_Projectile_Impact
 maxDistance=8300
 sampleRate=48000
 soundClassObject=&Fort_Impacts
 concurrencyOverrides {bLimitToOwner=true, maxCount=3}
 concurrencySet [&/Conc_Mistraven_ProjectileImpact_3P]
}
