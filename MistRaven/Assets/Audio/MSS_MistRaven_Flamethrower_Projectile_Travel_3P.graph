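# Mono one-shot MetaSound for the third-person travel sound of MistRaven's flamethrower projectile.
# OnPlay starts the first of two wave players; ReplayTravelSound alternates between them through a counter, switch and deferred reset, allowing overlapping retriggers.
# Each draws a non-repeating random sample with randomized pitch and volume, then the mix is scaled by character affiliation.
# Either player finishing fires OnFinished after a short delay.
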MetaSoundSource /MistRaven/Assets/Audio/MSS_MistRaven_Flamethrower_Projectile_Travel_3P
inputs [
 Character.Affiliation {type=Enum:FortAudioAffiliation}
 Character.Health {type=Float}
 Character.UniqueID {type=Int32}
 Random.Pitch.Base {type=Float}
 Random.Pitch.Spread {type=Float}
 Random.Volume.Base {type=Float}
 Random.Volume.Spread {type=Float}
 ReplayTravelSound {type=Trigger}
 Sounds {type=WaveAsset:Array}
 UE.Source.OnPlay {type=Trigger}
]
interfaces [Fort.Character:1.0, UE.OutputFormat.Mono:1.0, UE.Source.OneShot:1.0, UE.Source:1.1]
outputs [
 "UE.OutputFormat.Mono.Audio:0" {type=Audio}
 UE.Source.OneShot.OnFinished {type=Trigger}
]
pages [Default {
  input_defaults {
   Character.Affiliation=0
   Character.Health=0
   Character.UniqueID=0
   Random.Pitch.Base=0
   Random.Pitch.Spread=2
   Random.Volume.Base=0.4
   Random.Volume.Spread=0.1
   ReplayTravelSound=null
   UE.Source.OnPlay=null
   Sounds [
    &/MistRaven_Flamethrower_Projectile_Travel_11
    &/MistRaven_Flamethrower_Projectile_Travel_10
    &/MistRaven_Flamethrower_Projectile_Travel_09
    &/MistRaven_Flamethrower_Projectile_Travel_08
    &/MistRaven_Flamethrower_Projectile_Travel_07
    &/MistRaven_Flamethrower_Projectile_Travel_06
    &/MistRaven_Flamethrower_Projectile_Travel_05
    &/MistRaven_Flamethrower_Projectile_Travel_04
    &/MistRaven_Flamethrower_Projectile_Travel_03
    &/MistRaven_Flamethrower_Projectile_Travel_02
    &/MistRaven_Flamethrower_Projectile_Travel_01
   ]
  }
  nodes [
   n1 = VariableDeferredAccessor.Trigger(Variable=InitVariable.Trigger().Variable)
   TriggerCounter_0 = UE.Trigger Counter(In=^ReplayTravelSound, Reset=@n1.Value)
   n2 = /FortMetasound/Patches/MSP_SwitchOnInt_4(Int=@TriggerCounter_0.Count, Compare=@TriggerCounter_0.On Trigger)
   n3 = UE.RandomFloat(
    Max=UE.Add:Float(PrimaryOperand=^Random.Volume.Base, AdditionalOperands=^Random.Volume.Spread).Out
    Next=UE.Reroute(Value=@n2.Out 2).Value
    Min=UE.Subtract:Float(PrimaryOperand=^Random.Volume.Base, AdditionalOperands=^Random.Volume.Spread).Out
   )
   ArrayRandomGet_1 = Array.Random Get:WaveAsset:Array(In Array=^Sounds, Next=@n2.Out 2) {also {defaults {Enable Shared State="<MetaSoundBuilderResult.SUCCEEDED: 0>", No Repeats=7, Weights []}}}
   WavePlayer_1 = UE.Wave Player:Mono(
    Play=@ArrayRandomGet_1.On Next
    Wave Asset=@ArrayRandomGet_1.Value
    Pitch Shift=UE.RandomFloat(
     Max=UE.Add:Float(PrimaryOperand=^Random.Pitch.Base, AdditionalOperands=^Random.Pitch.Spread).Out
     Next=UE.Reroute(Value=@n2.Out 2).Value
     Min=UE.Subtract:Float(PrimaryOperand=^Random.Pitch.Base, AdditionalOperands=^Random.Pitch.Spread).Out
    ).Value
   )
   TriggerAnyTriggerAny2_2 = "TriggerAny.Trigger Any (2)"(In 0=^UE.Source.OnPlay, In 1=@n2.Out 1)
   n4 = UE.RandomFloat(
    Max=UE.Add:Float(PrimaryOperand=^Random.Volume.Base, AdditionalOperands=^Random.Volume.Spread).Out
    Next=UE.Reroute(Value=@TriggerAnyTriggerAny2_2.Out).Value
    Min=UE.Subtract:Float(PrimaryOperand=^Random.Volume.Base, AdditionalOperands=^Random.Volume.Spread).Out
   )
   TriggerAnyTriggerAny2_0 = "TriggerAny.Trigger Any (2)"(In 0=^UE.Source.OnPlay, In 1=@n2.Out 1)
   ArrayRandomGet_0 = Array.Random Get:WaveAsset:Array(In Array=^Sounds, Next=@TriggerAnyTriggerAny2_0.Out) {also {defaults {Enable Shared State="<MetaSoundBuilderResult.SUCCEEDED: 0>", No Repeats=7, Weights []}}}
   TriggerAnyTriggerAny2_3 = "TriggerAny.Trigger Any (2)"(In 0=^UE.Source.OnPlay, In 1=@n2.Out 1)
   WavePlayer_0 = UE.Wave Player:Mono(
    Play=@ArrayRandomGet_0.On Next
    Wave Asset=@ArrayRandomGet_0.Value
    Pitch Shift=UE.RandomFloat(
     Max=UE.Add:Float(PrimaryOperand=^Random.Pitch.Base, AdditionalOperands=^Random.Pitch.Spread).Out
     Next=UE.Reroute(Value=@TriggerAnyTriggerAny2_3.Out).Value
     Min=UE.Subtract:Float(PrimaryOperand=^Random.Pitch.Base, AdditionalOperands=^Random.Pitch.Spread).Out
    ).Value
   )
   AudioMixerAudioMixerMono2_0 = "AudioMixer.Audio Mixer (Mono, 2)"(
    In 1=UE.Multiply:Audio by Float(AdditionalOperands=Max.Max:Float(A=@n3.Value) {also {defaults {B=0}}}.Value, PrimaryOperand=@WavePlayer_1.Out Mono).Out
    In 0=UE.Multiply:Audio by Float(AdditionalOperands=Max.Max:Float(A=@n4.Value).Value, PrimaryOperand=@WavePlayer_0.Out Mono).Out
   )
   Multiply_0 = UE.Multiply:Audio by Float(
    PrimaryOperand=@AudioMixerAudioMixerMono2_0.Out
    AdditionalOperands=&MSP_AffiliationMixFloat(Affiliation=^Character.Affiliation) {also {defaults {Non-Enemy=0.3}}}.Out
   )
   n5 = InitVariable.Trigger()
   n6 = InitVariable.Trigger()
   TriggerAnyTriggerAny2_1 = "TriggerAny.Trigger Any (2)"(
    In 1=VariableAccessor.Trigger(Variable=VariableMutator.Trigger(Variable=@n5.Variable, Value=@WavePlayer_1.On Finished).Variable).Value
    In 0=VariableAccessor.Trigger(Variable=VariableMutator.Trigger(Variable=@n6.Variable, Value=@WavePlayer_0.On Finished).Variable).Value
   )
   TriggerDelay_0 = UE.Trigger Delay(In=@TriggerAnyTriggerAny2_1.Out) {also {defaults {Delay Time=0.25}}}
   VariableMutator.Trigger(Value=@n2.Out 2, Variable=@n1.Variable)
  ]
  links [{dst="@UE.OutputFormat.Mono.Audio:0", src=Multiply_0.Out}, {dst="@UE.Source.OneShot.OnFinished", src=TriggerDelay_0.Out}]
 }]
properties {
 assetImportData=&self:AssetImportData
 attenuationSettings=&/SA_MistRaven_Flamethrower_Projectile
 bOverrideConcurrency=true
 maxDistance=3300
 sampleRate=48000
 soundClassObject=&Fort_WeaponMelee
 concurrencyOverrides {maxCount=4, resolutionRule=PreventNew}
}
