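# MetaSound patch for one mono polyphony voice that plays a one-shot wave asset with randomized volume.
# Voice Start triggers the Wave Player and a random gain within base ± spread, floored at zero, which scales the audio.
# PriorityControl drives Pitch Shift directly, while the random pitch result is left unconnected.
# Stop and OnFinished tie to the player; voice priority and key index are unwired.
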MetaSoundPatch /FortMetasound/Patches/MSP_RandomOneShot_Polyphony
inputs [
 Polyphony.Voice.Control.KeyIndex.Index {type=Int32}
 Polyphony.Voice.In Priority {type=Int32}
 Polyphony.Voice.Start {type=Trigger}
 Polyphony.Voice.Stop {type=Trigger}
 PriorityControl {type=Float}
 Random.Pitch.Base {type=Float}
 Random.Pitch.Spread {type=Float}
 Random.Volume.Base {type=Float}
 Random.Volume.Spread {type=Float}
 Voice.Control.Sound {type=WaveAsset}
]
interfaces [Polyphony.Voice.Control.KeyIndex:1.0, Polyphony.Voice:1.0, UE.OutputFormat.Mono:1.0]
outputs [
 Polyphony.Voice.OnFinished {type=Trigger}
 Polyphony.Voice.Out Priority {type=Int32}
 "UE.OutputFormat.Mono.Audio:0" {type=Audio}
]
pages [Default {
  input_defaults {
   Polyphony.Voice.Control.KeyIndex.Index=null
   Polyphony.Voice.In Priority=null
   Polyphony.Voice.Start=null
   Polyphony.Voice.Stop=null
   PriorityControl=0
   Random.Pitch.Base=0
   Random.Pitch.Spread=2
   Random.Volume.Base=0.3
   Random.Volume.Spread=0.1
   Voice.Control.Sound=null
  }
  nodes [
   InitVariable.Int32()
   n1 = UE.Add:Float(PrimaryOperand=^Random.Volume.Base, AdditionalOperands=^Random.Volume.Spread)
   n2 = UE.Subtract:Float(PrimaryOperand=^Random.Volume.Base, AdditionalOperands=^Random.Volume.Spread)
   WavePlayer_0 = UE.Wave Player:Mono(Play=^Polyphony.Voice.Start, Stop=^Polyphony.Voice.Stop, Pitch Shift=^PriorityControl, Wave Asset=^Voice.Control.Sound)
   Multiply_0 = UE.Multiply:Audio by Float(
    AdditionalOperands=Max.Max:Float(A=UE.RandomFloat(Next=^Polyphony.Voice.Start, Max=@n1.Out, Min=@n2.Out).Value) {also {defaults {B=0}}}.Value
    PrimaryOperand=@WavePlayer_0.Out Mono
   )
   UE.RandomFloat(
    Next=^Polyphony.Voice.Start
    Max=UE.Add:Float(PrimaryOperand=^Random.Pitch.Base, AdditionalOperands=^Random.Pitch.Spread).Out
    Min=UE.Subtract:Float(PrimaryOperand=^Random.Pitch.Base, AdditionalOperands=^Random.Pitch.Spread).Out
   )
  ]
  links [{dst="@UE.OutputFormat.Mono.Audio:0", src=Multiply_0.Out}, {dst="@Polyphony.Voice.OnFinished", src=WavePlayer_0.On Finished}]
 }]
