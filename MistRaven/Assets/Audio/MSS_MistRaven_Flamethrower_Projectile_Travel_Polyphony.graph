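# One-shot mono MetaSound for MistRaven's flamethrower projectile travel sound.
# Play or ReplayTravelSound picks a low-repeat random wave and triggers a two-voice MSP_RandomOneShot_Polyphony patch, so retriggers can overlap.
# Affiliation mixes set per-voice base volume and scale the output again.
# OnFinished is gated by an envelope follower check after a wave-length-plus-padding delay; the Random inputs and TriggerRepeat node are unwired.
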MetaSoundSource /MistRaven/Assets/Audio/MSS_MistRaven_Flamethrower_Projectile_Travel_Polyphony
inputs [
 Character.Affiliation {type=Enum:FortAudioAffiliation}
 Character.Health {type=Float}
 Character.UniqueID {type=Int32}
 In {type=Float}
 Random.Pitch.Base {type=Float}
 Random.Pitch.Spread {type=Float}
 Random.Volume.Base {type=Float}
 Random.Volume.Spread {type=Float}
 ReplayTravelSound {type=Trigger}
 Sounds {type=WaveAsset:Array}
 UE.Source.OnPlay {type=Trigger}
]
interfaces [Fort.Character:1.0, UE.OutputFormat.Mono:1.0, UE.Source.OneShot:1.0, UE.Source:1.1]
outputs [
 "UE.OutputFormat.Mono.Audio:0" {type=Audio}
 UE.Source.OneShot.OnFinished {type=Trigger}
]
pages [Default {
  input_defaults {
   Character.Affiliation=0
   Character.Health=0
   Character.UniqueID=0
   In=4.52
   Random.Pitch.Base=0
   Random.Pitch.Spread=2
   Random.Volume.Base=0.3
   Random.Volume.Spread=0.1
   ReplayTravelSound=null
   UE.Source.OnPlay=null
   Sounds [
    &/MistRaven_Flamethrower_Projectile_Travel_11
    &/MistRaven_Flamethrower_Projectile_Travel_10
    &/MistRaven_Flamethrower_Projectile_Travel_09
    &/MistRaven_Flamethrower_Projectile_Travel_08
    &/MistRaven_Flamethrower_Projectile_Travel_07
    &/MistRaven_Flamethrower_Projectile_Travel_06
    &/MistRaven_Flamethrower_Projectile_Travel_05
    &/MistRaven_Flamethrower_Projectile_Travel_04
    &/MistRaven_Flamethrower_Projectile_Travel_03
    &/MistRaven_Flamethrower_Projectile_Travel_02
    &/MistRaven_Flamethrower_Projectile_Travel_01
   ]
  }
  nodes [
   InitVariable.Int32()
   InitVariable.Int32()
   InitVariable.Int32()
   InitVariable.Trigger()
   InitVariable.Trigger()
   InitVariable.Trigger()
   InitVariable.Trigger()
   InitVariable.Trigger()
   InitVariable.Trigger()
   TriggerAnyTriggerAny2_0 = "TriggerAny.Trigger Any (2)"(In 1=^ReplayTravelSound, In 0=^UE.Source.OnPlay)
   ArrayRandomGet_0 = Array.Random Get:WaveAsset:Array(In Array=^Sounds, Next=@TriggerAnyTriggerAny2_0.Out) {also {defaults {Enable Shared State="<MetaSoundBuilderResult.SUCCEEDED: 0>", No Repeats=7, Weights []}}}
   n1 = MetasoundPolyphonyInternal.ConversionFloatToPolyphony::VoiceParameters(In=&MSP_AffiliationMixFloat(Affiliation=^Character.Affiliation) {also {defaults {Enemy=0.4, Local=0.3, Non-Enemy=0.4}}}.Out) {also {defaults {Parameter Name=Random.Volume.Base}, version [0,1]}}
   n2 = MetasoundPolyphonyInternal.ConversionWaveAssetToPolyphony::VoiceParameters(In=@ArrayRandomGet_0.Value) {also {defaults {Parameter Name=Voice.Control.Sound}, version [0,1]}}
   n3 = MetasoundPolyphonyInternal.PolyphonyPatchPlayer:Mono(
    Voice Control Stream=MetasoundPolyphonyInternal.TriggerOneShotVoiceController(
     Trigger=@ArrayRandomGet_0.On Next
     Voice Parameters=MetasoundPolyphonyInternal.VoiceParametersMerge2Inputs("1: Voice Params"=@n1.Voice Parameters, "0: Voice Params"=@n2.Voice Parameters) {also {version [0,1]}}.Voice Params
    ) {also {version [0,1]}}.Voice Control Stream
    Voice Pool Params=MetasoundPolyphonyInternal.VoicePoolParamsBuilder() {also {defaults {Voice Limit=2}, version [0,1]}}.Out
   ) {also {defaults {MetaSound Asset=/FortMetasound/Patches/MSP_RandomOneShot_Polyphony}, version [0,1]}}
   Multiply_0 = UE.Multiply:Audio by Float(
    AdditionalOperands=&MSP_AffiliationMixFloat(Affiliation=^Character.Affiliation) {also {defaults {Non-Enemy=0.3}}}.Out
    PrimaryOperand=@n3.Out Mono
   )
   EnvelopeFollower_0 = UE.Envelope Follower(In=@n3.Out Mono) {also {defaults {Attack Time=0, Release Time=0}}}
   n4 = InitVariable.Trigger()
   GetWaveDuration_0 = "UE.Get Wave Duration: "(Wave=@ArrayRandomGet_0.Value)
   TriggerDelay_0 = UE.Trigger Delay(
    Delay Time=UE.Add:Time(PrimaryOperand=@GetWaveDuration_0.Duration) {also {defaults {AdditionalOperands=0.1}}}.Out
    In=@ArrayRandomGet_0.On Next
   )
   TriggerCompareClamp_0 = TriggerCompare.Clamp:Float(
    A=@EnvelopeFollower_0.Envelope
    Compare=VariableAccessor.Trigger(Variable=VariableMutator.Trigger(Variable=@n4.Variable, Value=@TriggerDelay_0.Out).Variable).Value
   ) {also {defaults {B=0, Type=4}}}
   UE.TriggerRepeat() {also {defaults {Period=0.25}}}
  ]
  links [{dst="@UE.OutputFormat.Mono.Audio:0", src=Multiply_0.Out}, {dst="@UE.Source.OneShot.OnFinished", src=TriggerCompareClamp_0.True}]
 }]
properties {
 assetImportData=&self:AssetImportData
 attenuationSettings=&/SA_MistRaven_Flamethrower_Projectile
 bOverrideConcurrency=true
 maxDistance=3300
 sampleRate=48000
 soundClassObject=&Fort_WeaponMelee
 volume=0.6
 concurrencyOverrides {maxCount=4, resolutionRule=PreventNew}
}
