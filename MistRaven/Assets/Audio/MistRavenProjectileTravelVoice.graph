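# MetaSound patch for one mono voice, named for Mist Raven projectile travel, that plays an input wave with per-trigger variation.
# Play starts the wave player and rerolls pitch shift within base plus or minus spread.
# It also rerolls a gain, clamped at zero, from base minus spread to an upper bound lacking the base, which is unwired in that Add.
# The scaled audio and finished trigger are the outputs.
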MetaSoundPatch /MistRaven/Assets/Audio/MistRavenProjectileTravelVoice
inputs [
 Play {type=Trigger}
 Random.Pitch.Base {type=Float}
 Random.Pitch.Spread {type=Float}
 Random.Volume.Base {type=Float}
 Random.Volume.Spread {type=Float}
 Sound {type=WaveAsset}
]
outputs [
 On Finished {type=Trigger}
 Out Mono {type=Audio}
]
pages [Default {
  input_defaults {Play=null, Random.Pitch.Base=0, Random.Pitch.Spread=2, Random.Volume.Base=0.35, Random.Volume.Spread=0.1, Sound=null}
  nodes [
   n1 = UE.Add:Float(AdditionalOperands=^Random.Volume.Spread)
   n2 = UE.Reroute(Value=^Play)
   n3 = UE.Subtract:Float(PrimaryOperand=^Random.Volume.Base, AdditionalOperands=^Random.Volume.Spread)
   WavePlayer_0 = UE.Wave Player:Mono(
    Play=^Play
    Wave Asset=^Sound
    Pitch Shift=UE.RandomFloat(
     Next=^Play
     Max=UE.Add:Float(PrimaryOperand=^Random.Pitch.Base, AdditionalOperands=^Random.Pitch.Spread).Out
     Min=UE.Subtract:Float(PrimaryOperand=^Random.Pitch.Base, AdditionalOperands=^Random.Pitch.Spread).Out
    ).Value
   )
   Multiply_0 = UE.Multiply:Audio by Float(
    AdditionalOperands=Max.Max:Float(A=UE.RandomFloat(Max=@n1.Out, Next=@n2.Value, Min=@n3.Out).Value) {also {defaults {B=0}}}.Value
    PrimaryOperand=@WavePlayer_0.Out Mono
   )
  ]
  links [{dst="@Out Mono", src=Multiply_0.Out}, {dst="@On Finished", src=WavePlayer_0.On Finished}]
 }]
