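# Mono one-shot MetaSound for the MistRaven flamethrower projectile's travel sound, with its own attenuation and concurrency limit.
# Play or ReplayTravelSound picks a random non-repeating wave, and a trigger sequence alternates it between two travel-voice patches so replays can overlap.
# Affiliation sets each voice's base volume and scales the mixed output.
# Either voice finishing ends the source after a short delay.
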MetaSoundSource /MistRaven/Assets/Audio/MSS_MistRaven_Flamethrower_Projectile_Travel_V2
inputs [
 Character.Affiliation {type=Enum:FortAudioAffiliation}
 Character.Health {type=Float}
 Character.UniqueID {type=Int32}
 Random.Pitch.Base {type=Float}
 Random.Pitch.Spread {type=Float}
 Random.Volume.Base {type=Float}
 Random.Volume.Spread {type=Float}
 ReplayTravelSound {type=Trigger}
 Sounds {type=WaveAsset:Array}
 UE.Source.OnPlay {type=Trigger}
]
interfaces [Fort.Character:1.0, UE.OutputFormat.Mono:1.0, UE.Source.OneShot:1.0, UE.Source:1.1]
outputs [
 "UE.OutputFormat.Mono.Audio:0" {type=Audio}
 UE.Source.OneShot.OnFinished {type=Trigger}
]
pages [Default {
  input_defaults {
   Character.Affiliation=0
   Character.Health=0
   Character.UniqueID=0
   Random.Pitch.Base=0
   Random.Pitch.Spread=2
   Random.Volume.Base=0.35
   Random.Volume.Spread=0.1
   ReplayTravelSound=null
   UE.Source.OnPlay=null
   Sounds [
    &/MistRaven_Flamethrower_Projectile_Travel_11
    &/MistRaven_Flamethrower_Projectile_Travel_10
    &/MistRaven_Flamethrower_Projectile_Travel_09
    &/MistRaven_Flamethrower_Projectile_Travel_08
    &/MistRaven_Flamethrower_Projectile_Travel_07
    &/MistRaven_Flamethrower_Projectile_Travel_06
    &/MistRaven_Flamethrower_Projectile_Travel_05
    &/MistRaven_Flamethrower_Projectile_Travel_04
    &/MistRaven_Flamethrower_Projectile_Travel_03
    &/MistRaven_Flamethrower_Projectile_Travel_02
    &/MistRaven_Flamethrower_Projectile_Travel_01
   ]
  }
  nodes [
   InitVariable.Trigger()
   TriggerAnyTriggerAny2_1 = "TriggerAny.Trigger Any (2)"(In 1=^ReplayTravelSound, In 0=^UE.Source.OnPlay)
   ArrayRandomGet_0 = Array.Random Get:WaveAsset:Array(In Array=^Sounds, Next=@TriggerAnyTriggerAny2_1.Out) {also {defaults {Enable Shared State="<MetaSoundBuilderResult.SUCCEEDED: 0>", No Repeats=7, Weights []}}}
   n1 = &MSP_AffiliationMixFloat(Affiliation=^Character.Affiliation) {also {defaults {Enemy=0.4, Local=0.3, Non-Enemy=0.4}}}
   TriggerSequence2_0 = "UE.Trigger Sequence (2)"(In=@ArrayRandomGet_0.On Next)
   n2 = &/MistRavenProjectileTravelVoice(Sound=@ArrayRandomGet_0.Value, Random.Volume.Base=@n1.Out, Play=@TriggerSequence2_0.Out 0)
   n3 = &MistRavenProjectileTravelVoice(Sound=@ArrayRandomGet_0.Value, Random.Volume.Base=@n1.Out, Play=@TriggerSequence2_0.Out 1)
   AudioMixerAudioMixerMono2_0 = "AudioMixer.Audio Mixer (Mono, 2)"(In 0=@n2.Out Mono, In 1=@n3.Out Mono)
   Multiply_0 = UE.Multiply:Audio by Float(
    PrimaryOperand=@AudioMixerAudioMixerMono2_0.Out
    AdditionalOperands=&MSP_AffiliationMixFloat(Affiliation=^Character.Affiliation) {also {defaults {Non-Enemy=0.3}}}.Out
   )
   n4 = InitVariable.Trigger()
   n5 = InitVariable.Trigger()
   TriggerAnyTriggerAny2_0 = "TriggerAny.Trigger Any (2)"(
    In 1=VariableAccessor.Trigger(Variable=VariableMutator.Trigger(Variable=@n4.Variable, Value=@n3.On Finished).Variable).Value
    In 0=VariableAccessor.Trigger(Variable=VariableMutator.Trigger(Variable=@n5.Variable, Value=@n2.On Finished).Variable).Value
   )
   TriggerDelay_0 = UE.Trigger Delay(In=@TriggerAnyTriggerAny2_0.Out) {also {defaults {Delay Time=0.25}}}
  ]
  links [{dst="@UE.OutputFormat.Mono.Audio:0", src=Multiply_0.Out}, {dst="@UE.Source.OneShot.OnFinished", src=TriggerDelay_0.Out}]
 }]
properties {
 assetImportData=&self:AssetImportData
 attenuationSettings=&/SA_MistRaven_Flamethrower_Projectile
 bOverrideConcurrency=true
 maxDistance=3300
 sampleRate=48000
 soundClassObject=&Fort_WeaponMelee
 concurrencyOverrides {maxCount=4, resolutionRule=PreventNew}
}
