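# One-shot stereo MetaSound source for the start sound of the MistRaven flamethrower.
# On play it picks one of three start waves at random and randomizes pitch and volume from base and spread inputs, though default volume spread is zero.
# An affiliation mix patch scales both channels, weighting enemies well above non-enemies, and wave completion fires OnFinished.
# It uses its own attenuation and concurrency assets and the melee weapon sound class.
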MetaSoundSource /MistRaven/Assets/Audio/MSS_MistRaven_Flamethrower_Start
inputs [
 Character.Affiliation {type=Enum:FortAudioAffiliation}
 Character.Health {type=Float}
 Character.UniqueID {type=Int32}
 Random.Pitch.Base {type=Float}
 Random.Pitch.Spread {type=Float}
 Random.Volume.Base {type=Float}
 Random.Volume.Spread {type=Float}
 Sounds {type=WaveAsset:Array}
 UE.Source.OnPlay {type=Trigger}
]
interfaces [Fort.Character:1.0, UE.OutputFormat.Stereo:1.0, UE.Source.OneShot:1.0, UE.Source:1.1]
outputs [
 "UE.OutputFormat.Stereo.Audio:0" {type=Audio}
 "UE.OutputFormat.Stereo.Audio:1" {type=Audio}
 UE.Source.OneShot.OnFinished {type=Trigger}
]
pages [Default {
  input_defaults {
   Character.Affiliation=0
   Character.Health=0
   Character.UniqueID=0
   Random.Pitch.Base=0
   Random.Pitch.Spread=1
   Random.Volume.Base=0.85
   Random.Volume.Spread=0
   UE.Source.OnPlay=null
   Sounds [
    &/MistRaven_Flamethrower_Start_01
    &/MistRaven_Flamethrower_Start_02
    &/MistRaven_Flamethrower_Start_03
   ]
  }
  nodes [
   n1 = &MSP_AffiliationMixFloat(Affiliation=^Character.Affiliation) {also {defaults {Enemy=0.8, Non-Enemy=0.2}}}
   n2 = Max.Max:Float(A=UE.RandomFloat(
     Max=UE.Add:Float(PrimaryOperand=^Random.Volume.Base, AdditionalOperands=^Random.Volume.Spread).Out
     Next=UE.Reroute(Value=^UE.Source.OnPlay).Value
     Min=UE.Subtract:Float(PrimaryOperand=^Random.Volume.Base, AdditionalOperands=^Random.Volume.Spread).Out
    ).Value)
   ArrayRandomGet_0 = Array.Random Get:WaveAsset:Array(In Array=^Sounds, Next=^UE.Source.OnPlay) {also {defaults {Enable Shared State="<MetaSoundBuilderResult.SUCCEEDED: 0>", Weights []}}}
   WavePlayer_0 = UE.Wave Player:Stereo(
    Play=@ArrayRandomGet_0.On Next
    Wave Asset=@ArrayRandomGet_0.Value
    Pitch Shift=UE.RandomFloat(
     Max=UE.Add:Float(PrimaryOperand=^Random.Pitch.Base, AdditionalOperands=^Random.Pitch.Spread).Out
     Next=UE.Reroute(Value=^UE.Source.OnPlay).Value
     Min=UE.Subtract:Float(PrimaryOperand=^Random.Pitch.Base, AdditionalOperands=^Random.Pitch.Spread).Out
    ).Value
   ) {also {version [1,1]}}
   Multiply_0 = UE.Multiply:Audio by Float(
    AdditionalOperands=@n1.Out
    PrimaryOperand=UE.Multiply:Audio by Float(AdditionalOperands=@n2.Value, PrimaryOperand=@WavePlayer_0.Out Left).Out
   )
   Multiply_3 = UE.Multiply:Audio by Float(
    AdditionalOperands=@n1.Out
    PrimaryOperand=UE.Multiply:Audio by Float(AdditionalOperands=@n2.Value, PrimaryOperand=@WavePlayer_0.Out Right).Out
   )
  ]
  links [
   {dst="@UE.OutputFormat.Stereo.Audio:0", src=Multiply_0.Out}
   {dst="@UE.OutputFormat.Stereo.Audio:1", src=Multiply_3.Out}
   {dst="@UE.Source.OneShot.OnFinished", src=WavePlayer_0.On Finished}
  ]
 }]
properties {
 assetImportData=&self:AssetImportData
 attenuationSettings=&/SA_MistRaven_Flamethrower_Attack_Start
 maxDistance=10750
 numChannels=2
 outputFormat=Stereo
 sampleRate=48000
 soundClassObject=&Fort_WeaponMelee
 concurrencySet [&/CONC_MistRaven_Flamethrower_Start]
}
