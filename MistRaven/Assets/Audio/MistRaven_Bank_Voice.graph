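# MetaSound patch for a Mist Raven bank voice, playing random stereo one-shots driven by movement and azimuth speed.
# Two MSP_OneshotGate instances test the speeds against a shared threshold and retrigger delay, and either can start or release an ADSR envelope.
# Attack picks and plays a non-repeating random wave, while release also resets the picker.
# The envelope scales both channels, with Fade setting attack and release times.
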MetaSoundPatch /MistRaven/Assets/Audio/MistRaven_Bank_Voice
inputs [
 AzimuthSpeed {type=Float}
 Fade {type=Float}
 MovementSpeed {type=Float}
 OneShotSounds {type=WaveAsset:Array}
 OneshotThreshold {type=Float}
 RetriggerDelay {type=Time}
]
outputs [
 On Finished {type=Trigger}
 On Play {type=Trigger}
 Out L {type=Audio}
 Out R {type=Audio}
]
pages [Default {
  input_defaults {AzimuthSpeed=1, Fade=0.25, MovementSpeed=1, OneshotThreshold=0.1, RetriggerDelay=1, OneShotSounds []}
  nodes [
   n1 = &MSP_OneshotGate(Value=^MovementSpeed, OneshotThreshold=^OneshotThreshold, RetriggerDelay=^RetriggerDelay) {also {defaults {EnterCondition=2, ExitCondition=3}}}
   n2 = &MSP_OneshotGate(Value=^AzimuthSpeed, OneshotThreshold=^OneshotThreshold, RetriggerDelay=^RetriggerDelay) {also {defaults {EnterCondition=3}}}
   TriggerAnyTriggerAny2_0 = "TriggerAny.Trigger Any (2)"(In 0=@n1.OnGateExit, In 1=@n2.OnGateExit)
   TriggerAnyTriggerAny2_1 = "TriggerAny.Trigger Any (2)"(In 0=@n1.OnGateEnter, In 1=@n2.OnGateEnter)
   ADSREnvelopeADSREnvelope_0 = ADSR Envelope.ADSR Envelope:Float(
    Release Time=UE.ConversionFloatToTime(In=^Fade).Out
    Attack Time=UE.ConversionFloatToTime(In=^Fade).Out
    Trigger Release=@TriggerAnyTriggerAny2_0.Out
    Trigger Attack=@TriggerAnyTriggerAny2_1.Out
   )
   ArrayRandomGet_0 = Array.Random Get:WaveAsset:Array(In Array=^OneShotSounds, Next=@ADSREnvelopeADSREnvelope_0.On Attack Triggered, Reset=@ADSREnvelopeADSREnvelope_0.On Release Triggered) {also {defaults {No Repeats=3, Weights []}}}
   WavePlayer_0 = UE.Wave Player:Stereo(Play=@ArrayRandomGet_0.On Next, Wave Asset=@ArrayRandomGet_0.Value)
   Multiply_0 = UE.Multiply:Audio by Float(AdditionalOperands=@ADSREnvelopeADSREnvelope_0.Out Envelope, PrimaryOperand=@WavePlayer_0.Out Right)
   Multiply_1 = UE.Multiply:Audio by Float(AdditionalOperands=@ADSREnvelopeADSREnvelope_0.Out Envelope, PrimaryOperand=@WavePlayer_0.Out Left)
  ]
  links [
   {dst="@Out R", src=Multiply_0.Out}
   {dst="@Out L", src=Multiply_1.Out}
   {dst="@On Finished", src=WavePlayer_0.On Finished}
   {dst="@On Play", src=WavePlayer_0.On Play}
  ]
 }]
