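# MetaSound patch for MistRaven movement audio that turns speed, direction and time inputs into smoothed control values, including tremolo rate and depth.
# Speed is remapped against TopSpeed with a SpeedFloor minimum, then smoothed to drive both tremolo ranges.
# Deferred variables supply previous speed and time to derive a scaled, smoothed SpeedDelta.
# Azimuth and inverted, smoothed elevation map to signed unit ranges, while an OnPlay-triggered ADSR outputs the fade-in.
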MetaSoundPatch /MistRaven/Assets/Audio/MistRaven_Movement_Control
inputs [
 FadeInTime {type=Time}
 OnPlay {type=Trigger}
 Speed {type=Float}
 SpeedAzimuth {type=Float}
 SpeedElevation {type=Float}
 SpeedFloor {type=Float}
 Time {type=Float}
 TopSpeed {type=Float}
 TremoloDepthMax {type=Float}
 TremoloDepthMin {type=Float}
 TremoloRateMax {type=Float}
 TremoloRateMin {type=Float}
]
outputs [
 AzimuthSpeed {type=Float}
 ElevationSpeed {type=Float}
 FadeInTime {type=Float}
 NormalizedSpeed {type=Float}
 SpeedDelta {type=Float}
 TremoloDepth {type=Float}
 TremoloRate {type=Float}
]
pages [Default {
  input_defaults {
   FadeInTime=1
   OnPlay=null
   Speed=0
   SpeedAzimuth=0
   SpeedElevation=0
   SpeedFloor=0
   Time=0
   TopSpeed=1
   TremoloDepthMax=0.6
   TremoloDepthMin=0.8
   TremoloRateMax=12
   TremoloRateMin=1
  }
  nodes [
   ADSREnvelopeADSREnvelope_0 = ADSR Envelope.ADSR Envelope:Float(Attack Time=^FadeInTime, Trigger Attack=^OnPlay) {also {defaults {Sustain Level=1}}}
   n1 = MapRange.MapRange:Float(In=^Speed, Out Range A=^SpeedFloor, In Range B=^TopSpeed)
   n2 = VariableDeferredAccessor.Float(Variable=InitVariable.Float().Variable)
   n3 = UE.Subtract:Float(
    PrimaryOperand=VariableAccessor.Float(Variable=VariableMutator.Float(Value=@n1.Out Value, Variable=@n2.Variable).Variable).Value
    AdditionalOperands=@n2.Value
   )
   n4 = VariableDeferredAccessor.Float(Variable=InitVariable.Float().Variable)
   n5 = UE.Subtract:Float(
    PrimaryOperand=VariableAccessor.Float(Variable=VariableMutator.Float(Value=^Time, Variable=@n4.Variable).Variable).Value
    AdditionalOperands=@n4.Value
   )
   InterpTo_1 = UE.InterpTo:Audio(Target=UE.Multiply:Float(PrimaryOperand=UE.Divide:Float(PrimaryOperand=@n3.Out, AdditionalOperands=@n5.Out).Out) {also {defaults {AdditionalOperands=1000}}}.Out) {also {defaults {Interp Time=1}}}
   InterpTo_2 = UE.InterpTo:Audio(Target=MapRange.MapRange:Float(In=^SpeedElevation) {also {defaults {In Range A=-80, In Range B=80, Out Range A=1, Out Range B=-1}}}.Out Value) {also {defaults {Interp Time=0.2}}}
   InterpTo_0 = UE.InterpTo:Audio(Target=@n1.Out Value) {also {defaults {Interp Time=0.5}}}
   MapRangeMapRange_0 = MapRange.MapRange:Float(Out Range B=^TremoloRateMax, Out Range A=^TremoloRateMin, In=@InterpTo_0.Value)
   MapRangeMapRange_2 = MapRange.MapRange:Float(In=^SpeedAzimuth) {also {defaults {In Range A=-90, In Range B=90, Out Range A=-1}}}
   MapRangeMapRange_4 = MapRange.MapRange:Float(Out Range B=^TremoloDepthMax, Out Range A=^TremoloDepthMin, In=@InterpTo_0.Value)
   VariableMutator.Float(Variable=InitVariable.Float().Variable, Value=@InterpTo_0.Value)
  ]
  links [
   {dst="@FadeInTime", src=ADSREnvelopeADSREnvelope_0.Out Envelope}
   {dst="@NormalizedSpeed", src=InterpTo_0.Value}
   {dst="@SpeedDelta", src=InterpTo_1.Value}
   {dst="@ElevationSpeed", src=InterpTo_2.Value}
   {dst="@TremoloRate", src=MapRangeMapRange_0.Out Value}
   {dst="@AzimuthSpeed", src=MapRangeMapRange_2.Out Value}
   {dst="@TremoloDepth", src=MapRangeMapRange_4.Out Value}
  ]
 }]
