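# MetaSound patch for the MistRaven GrimStrut vocal voice, layering a delayed random start one-shot with a matching loop.
# The one-shot's random index selects the loop, which starts slightly later under an ADSR attack with LFO tremolo.
# EndBurstEnvelope drives the loop's tremolo rate, pitch, gain and high-pass cutoff.
# A stereo mixer combines both layers, with NormalizedSpeed boosting the loop's level.
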MetaSoundPatch /MistRaven/Assets/Audio/MistRaven_Vocal_Voice
inputs [
 EndBurstEnvelope {type=Float}
 NormalizedSpeed {type=Float}
 On.Play {type=Trigger}
 TremoloDepth {type=Float}
 TremoloRate {type=Float}
 VocalLoops {type=WaveAsset:Array}
 VocalOneShots {type=WaveAsset:Array}
]
outputs [
 Out L {type=Audio}
 Out R {type=Audio}
]
pages [Default {
  input_defaults {
   EndBurstEnvelope=0
   NormalizedSpeed=0
   On.Play=null
   TremoloDepth=0
   TremoloRate=0
   VocalLoops [&/MistRaven_GrimStrut_Vocal_Loop_01, &/MistRaven_GrimStrut_Vocal_Loop_02]
   VocalOneShots [&/MistRaven_GrimStrut_Vocal_Start_01, &/MistRaven_GrimStrut_Vocal_Start_02]
  }
  nodes [
   n1 = UE.Add:Float(PrimaryOperand=UE.Multiply:Float(PrimaryOperand=^EndBurstEnvelope) {also {defaults {AdditionalOperands=1.5}}}.Out) {also {defaults {AdditionalOperands=1}}}
   TriggerDelay_1 = UE.Trigger Delay(In=^On.Play) {also {defaults {Delay Time=0.813}}}
   ADSREnvelopeADSREnvelope_0 = ADSR Envelope.ADSR Envelope:Float(Trigger Attack=@TriggerDelay_1.Out) {also {defaults {Attack Time=0.1, Decay Time=0.1, Sustain Level=0.8}}}
   n2 = UE.LFO:Audio(
    Frequency=MapRange.MapRange:Float(
     In=^EndBurstEnvelope
     Out Range B=UE.Multiply:Float(PrimaryOperand=^TremoloRate) {also {defaults {AdditionalOperands=1.25}}}.Out
     Out Range A=UE.Multiply:Float(PrimaryOperand=^TremoloRate) {also {defaults {AdditionalOperands=0.5}}}.Out
    ).Out Value
    Min Value=UE.Multiply:Float(PrimaryOperand=^TremoloDepth) {also {defaults {AdditionalOperands=1.2}}}.Out
    Sync=UE.Reroute(Value=^On.Play).Value
   ) {also {defaults {Shape=3}}}
   TriggerDelay_0 = UE.Trigger Delay(In=^On.Play) {also {defaults {Delay Time=0.613}}}
   ArrayRandomGet_0 = Array.Random Get:WaveAsset:Array(In Array=^VocalOneShots, Next=@TriggerDelay_0.Out) {also {defaults {Enable Shared State="<MetaSoundBuilderResult.SUCCEEDED: 0>", Weights []}}}
   n3 = VariableMutator.Int32(Value=@ArrayRandomGet_0.Index, Variable=InitVariable.Int32().Variable)
   WavePlayer_0 = UE.Wave Player:Stereo(
    Pitch Shift=UE.Add:Float(AdditionalOperands=^EndBurstEnvelope).Out
    Wave Asset=Array.Get:WaveAsset:Array(Array=^VocalLoops, Trigger=@TriggerDelay_1.Out, Index=VariableAccessor.Int32(Variable=@n3.Variable).Value).Element
    Play=@TriggerDelay_1.Out
   ) {also {defaults {Loop="<MetaSoundBuilderResult.SUCCEEDED: 0>"}}}
   n4 = UE.Multiply:Audio by Float(AdditionalOperands=@n2.Out, PrimaryOperand=@WavePlayer_0.Out Left)
   n5 = UE.Multiply:Float(PrimaryOperand=^EndBurstEnvelope) {also {defaults {AdditionalOperands=1800}}}
   BiquadFilter_0 = UE.Biquad Filter:Audio(
    In=UE.Multiply:Audio by Float(
     AdditionalOperands=@n1.Out
     PrimaryOperand=UE.Multiply:Audio by Float(AdditionalOperands=@ADSREnvelopeADSREnvelope_0.Out Envelope, PrimaryOperand=@n4.Out).Out
    ).Out
    Cutoff Frequency=@n5.Out
   ) {also {defaults {Bandwidth=0.75, Type=1}}}
   n6 = UE.Multiply:Audio by Float(AdditionalOperands=@n2.Out, PrimaryOperand=@WavePlayer_0.Out Right)
   BiquadFilter_1 = UE.Biquad Filter:Audio(
    Cutoff Frequency=@n5.Out
    In=UE.Multiply:Audio by Float(
     AdditionalOperands=@n1.Out
     PrimaryOperand=UE.Multiply:Audio by Float(AdditionalOperands=@ADSREnvelopeADSREnvelope_0.Out Envelope, PrimaryOperand=@n6.Out).Out
    ).Out
   ) {also {defaults {Bandwidth=0.75, Type=1}}}
   WavePlayer_1 = UE.Wave Player:Stereo(Play=@ArrayRandomGet_0.On Next, Wave Asset=@ArrayRandomGet_0.Value)
   AudioMixerAudioMixerStereo2_0 = "AudioMixer.Audio Mixer (Stereo, 2)"(
    Gain 1=UE.Add:Float(PrimaryOperand=UE.Multiply:Float(PrimaryOperand=^NormalizedSpeed) {also {defaults {AdditionalOperands=2}}}.Out) {also {defaults {AdditionalOperands=1}}}.Out
    In 1 L=@BiquadFilter_0.Out
    In 1 R=@BiquadFilter_1.Out
    In 0 L=@WavePlayer_1.Out Left
    In 0 R=@WavePlayer_1.Out Right
   )
   InitVariable.Float()
   InitVariable.Float()
   InitVariable.Float()
  ]
  links [{dst="@Out L", src=AudioMixerAudioMixerStereo2_0.Out L}, {dst="@Out R", src=AudioMixerAudioMixerStereo2_0.Out R}]
 }]
