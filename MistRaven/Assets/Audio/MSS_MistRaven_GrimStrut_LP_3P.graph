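# Third-person looping movement sound for the MistRaven GrimStrut, layering traversal, banking and wind, scaled by affiliation.
# MistRaven_Movement_Control turns speed and direction into values driving loop pitch, wind level, banking one-shots and tremolo.
# An LFO modulates the base loop, whose left channel feeds a modulated delay mixed by speed.
# An ending-burst envelope raises wind pitch; it also feeds the vocal voice, whose output is unconnected to the mixer.
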MetaSoundSource /MistRaven/Assets/Audio/MSS_MistRaven_GrimStrut_LP_3P
inputs [
 Bank.FadeIn {type=Float}
 BankThreshold {type=Float}
 Character.Affiliation {type=Enum:FortAudioAffiliation}
 Character.Health {type=Float}
 Character.UniqueID {type=Int32}
 EndingBurstTrigger {type=Trigger}
 Mixing.BankingGain {type=Float}
 Mixing.BaseLayerGain {type=Float}
 Mixing.MasterVolume {type=Float}
 Mixing.VocalLoopGain {type=Float}
 Mixing.WindGain {type=Float}
 Movement.Speed {type=Float}
 Movement.SpeedAzimuth {type=Float}
 Movement.SpeedElevation {type=Float}
 OneShotRetriggerDelay {type=Time}
 Sounds.BankOneShots {type=WaveAsset:Array}
 Sounds.BaseLoop {type=WaveAsset}
 Sounds.ElevationLoop {type=WaveAsset}
 UE.Source.OnPlay {type=Trigger}
 VocalLoops {type=WaveAsset:Array}
 VocalOneShots {type=WaveAsset:Array}
]
interfaces [Fort.Character:1.0, Fort.Movement:1.0, UE.OutputFormat.Stereo:1.0, UE.Source.OneShot:1.0, UE.Source:1.1]
outputs [
 "UE.OutputFormat.Stereo.Audio:0" {type=Audio}
 "UE.OutputFormat.Stereo.Audio:1" {type=Audio}
 UE.Source.OneShot.OnFinished {type=Trigger}
]
pages [Default {
  input_defaults {
   Bank.FadeIn=0.1
   BankThreshold=0.5
   Character.Affiliation=0
   Character.Health=0
   Character.UniqueID=0
   EndingBurstTrigger=null
   Mixing.BankingGain=0.75
   Mixing.BaseLayerGain=1
   Mixing.MasterVolume=0.876
   Mixing.VocalLoopGain=0
   Mixing.WindGain=0.75
   Movement.Speed=0
   Movement.SpeedAzimuth=-90
   Movement.SpeedElevation=0
   OneShotRetriggerDelay=0.25
   Sounds.BaseLoop=&/MistRaven_GrimStrut_Traversal_Loop_01
   Sounds.ElevationLoop=&/MistRaven_GrimStrut_Wind_Loop_01
   UE.Source.OnPlay=null
   Sounds.BankOneShots [
    &/MistRaven_GrimStrut_Traversal_Turns_06
    &/MistRaven_GrimStrut_Traversal_Turns_05
    &/MistRaven_GrimStrut_Traversal_Turns_04
    &/MistRaven_GrimStrut_Traversal_Turns_03
    &/MistRaven_GrimStrut_Traversal_Turns_02
    &/MistRaven_GrimStrut_Traversal_Turns_01
   ]
   VocalLoops [&/MistRaven_GrimStrut_Vocal_Loop_01, &/MistRaven_GrimStrut_Vocal_Loop_02]
   VocalOneShots [&/MistRaven_GrimStrut_Vocal_Start_01, &/MistRaven_GrimStrut_Vocal_Start_02]
  }
  nodes [
   ADSREnvelopeADSREnvelope_0 = ADSR Envelope.ADSR Envelope:Float(Trigger Attack=^EndingBurstTrigger) {also {defaults {Attack Time=0.6, Sustain Level=1}}}
   n1 = VariableAccessor.Float(Variable=VariableMutator.Float(Value=@ADSREnvelopeADSREnvelope_0.Out Envelope, Variable=InitVariable.Float().Variable).Variable)
   n2 = &/MistRaven_Movement_Control(
    Speed=^Movement.Speed
    SpeedAzimuth=^Movement.SpeedAzimuth
    SpeedElevation=^Movement.SpeedElevation
    OnPlay=^UE.Source.OnPlay
    TremoloDepthMax=VariableAccessor.Float(Variable=InitVariable.Float().Variable).Value
    TremoloRateMax=VariableAccessor.Float(Variable=InitVariable.Float().Variable).Value
    TremoloDepthMin=VariableAccessor.Float(Variable=InitVariable.Float().Variable).Value
    TremoloRateMin=VariableAccessor.Float(Variable=InitVariable.Float().Variable).Value
   )
   n3 = VariableAccessor.Float(Variable=VariableMutator.Float(Variable=InitVariable.Float().Variable, Value=@n2.TremoloDepth).Variable)
   n4 = VariableAccessor.Float(Variable=VariableMutator.Float(Variable=InitVariable.Float().Variable, Value=@n2.NormalizedSpeed).Variable)
   n5 = VariableAccessor.Float(Variable=@n4.Variable)
   n6 = VariableAccessor.Float(Variable=@n5.Variable)
   n7 = VariableAccessor.Float(Variable=@n6.Variable)
   n8 = VariableAccessor.Float(Variable=VariableMutator.Float(Variable=InitVariable.Float().Variable, Value=@n2.TremoloRate).Variable)
   &/MistRaven_Vocal_Voice(
    VocalLoops=^VocalLoops
    VocalOneShots=^VocalOneShots
    EndBurstEnvelope=VariableAccessor.Float(Variable=@n1.Variable).Value
    TremoloDepth=VariableAccessor.Float(Variable=@n3.Variable).Value
    NormalizedSpeed=@n7.Value
    TremoloRate=VariableAccessor.Float(Variable=@n8.Variable).Value
   )
   n9 = &MSP_AffiliationMixFloat(Affiliation=^Character.Affiliation) {also {defaults {Enemy=0.8, Non-Enemy=0.35}}}
   n10 = UE.LFO:Audio(Sync=^UE.Source.OnPlay, Min Value=@n3.Value, Frequency=@n8.Value)
   n11 = VariableAccessor.Float(Variable=VariableMutator.Float(Variable=InitVariable.Float().Variable, Value=@n2.ElevationSpeed).Variable)
   n12 = VariableAccessor.Float(Variable=@n11.Variable)
   TriggerDelay_1 = UE.Trigger Delay(In=^UE.Source.OnPlay) {also {defaults {Delay Time=0.7}}}
   WavePlayer_0 = UE.Wave Player:Stereo(
    Wave Asset=^Sounds.BaseLoop
    Pitch Shift=UE.Add:Float(
     AdditionalOperands=UE.Multiply:Float(PrimaryOperand=@n12.Value) {also {defaults {AdditionalOperands=-4}}}.Out
     PrimaryOperand=UE.Multiply:Float(PrimaryOperand=@n4.Value) {also {defaults {AdditionalOperands=2}}}.Out
    ).Out
    Start Time=&MSP_WaveGetRandomStartTime(Sound=^Sounds.BaseLoop).Value
    Play=@TriggerDelay_1.Out
   ) {also {defaults {Loop="<MetaSoundBuilderResult.SUCCEEDED: 0>"}}}
   n13 = UE.Multiply:Audio by Float(AdditionalOperands=@n10.Out, PrimaryOperand=@WavePlayer_0.Out Left)
   n14 = UE.Delay:Audio(
    Delay Time=UE.ConversionFloatToTime(In=MapRange.MapRange:Float(In=@n10.Out) {also {defaults {Out Range A=0.1, Out Range B=0.3}}}.Out Value).Out
    In=@n13.Out
   ) {also {defaults {Feedback=0.2}}}
   n15 = VariableMutator.Float(Variable=InitVariable.Float().Variable, Value=@n2.AzimuthSpeed)
   n16 = &/MistRaven_Bank_Voice(
    Fade=^Bank.FadeIn
    OneshotThreshold=^BankThreshold
    RetriggerDelay=^OneShotRetriggerDelay
    OneShotSounds=^Sounds.BankOneShots
    AzimuthSpeed=AbsoluteValue.Abs:Float(Input=VariableAccessor.Float(Variable=@n15.Variable).Value).Value
    MovementSpeed=VariableAccessor.Float(Variable=@n7.Variable).Value
   )
   n17 = UE.Add:Float(PrimaryOperand=UE.Multiply:Float(PrimaryOperand=@n6.Value) {also {defaults {AdditionalOperands=0.5}}}.Out) {also {defaults {AdditionalOperands=0.5}}}
   n18 = UE.Multiply:Float(PrimaryOperand=@n1.Value) {also {defaults {AdditionalOperands=3}}}
   n19 = UE.Multiply:Float(PrimaryOperand=@n6.Value) {also {defaults {AdditionalOperands=1.5}}}
   TriggerDelay_0 = UE.Trigger Delay(In=^UE.Source.OnPlay) {also {defaults {Delay Time=0.7}}}
   WavePlayer_1 = UE.Wave Player:Stereo(
    Wave Asset=^Sounds.ElevationLoop
    Pitch Shift=UE.Add:Float(
     AdditionalOperands=UE.Add:Float(AdditionalOperands=@n18.Out, PrimaryOperand=@n19.Out).Out
     PrimaryOperand=UE.Multiply:Float(PrimaryOperand=@n11.Value) {also {defaults {AdditionalOperands=-2}}}.Out
    ).Out
    Start Time=&MSP_WaveGetRandomStartTime(Sound=^Sounds.ElevationLoop).Value
    Play=@TriggerDelay_0.Out
   ) {also {defaults {Loop="<MetaSoundBuilderResult.SUCCEEDED: 0>"}}}
   n20 = InitVariable.Float()
   n21 = InitVariable.Float()
   n22 = InitVariable.Float()
   n23 = InitVariable.Float()
   AudioMixerAudioMixerStereo5_0 = "AudioMixer.Audio Mixer (Stereo, 5)"(
    In 4 L=@n14.Out
    In 4 R=@n14.Out
    In 1 L=@n16.Out L
    In 1 R=@n16.Out R
    In 0 R=UE.Multiply:Audio by Float(AdditionalOperands=@n10.Out, PrimaryOperand=@WavePlayer_0.Out Right).Out
    In 2 R=UE.Multiply:Audio by Float(AdditionalOperands=@n17.Out, PrimaryOperand=@WavePlayer_1.Out Right).Out
    Gain 4=UE.Multiply:Float(PrimaryOperand=@n5.Value) {also {defaults {AdditionalOperands=0.5}}}.Out
    In 0 L=@n13.Out
    In 2 L=UE.Multiply:Audio by Float(AdditionalOperands=@n17.Out, PrimaryOperand=@WavePlayer_1.Out Left).Out
    Gain 1=VariableAccessor.Float(Variable=VariableMutator.Float(Value=^Mixing.BankingGain, Variable=@n20.Variable).Variable).Value
    Gain 3=VariableAccessor.Float(Variable=VariableMutator.Float(Value=^Mixing.VocalLoopGain, Variable=@n21.Variable).Variable).Value
    Gain 2=VariableAccessor.Float(Variable=VariableMutator.Float(Value=^Mixing.WindGain, Variable=@n22.Variable).Variable).Value
    Gain 0=VariableAccessor.Float(Variable=VariableMutator.Float(Value=^Mixing.BaseLayerGain, Variable=@n23.Variable).Variable).Value
   )
   n24 = VariableAccessor.Float(Variable=VariableMutator.Float(Value=^Mixing.MasterVolume, Variable=InitVariable.Float().Variable).Variable)
   Multiply_4 = UE.Multiply:Audio by Float(
    AdditionalOperands=@n9.Out
    PrimaryOperand=UE.Multiply:Audio by Float(PrimaryOperand=@AudioMixerAudioMixerStereo5_0.Out R, AdditionalOperands=@n24.Value).Out
   )
   Multiply_5 = UE.Multiply:Audio by Float(
    AdditionalOperands=@n9.Out
    PrimaryOperand=UE.Multiply:Audio by Float(PrimaryOperand=@AudioMixerAudioMixerStereo5_0.Out L, AdditionalOperands=@n24.Value).Out
   )
   VariableMutator.Float(Variable=InitVariable.Float().Variable, Value=@n2.SpeedDelta)
   VariableMutator.Float(Variable=InitVariable.Float().Variable, Value=@n2.FadeInTime)
  ]
  links [
   {dst="@UE.OutputFormat.Stereo.Audio:1", src=Multiply_4.Out}
   {dst="@UE.OutputFormat.Stereo.Audio:0", src=Multiply_5.Out}
   {dst="@UE.Source.OneShot.OnFinished", src=WavePlayer_0.On Finished}
  ]
 }]
properties {
 assetImportData=&self:AssetImportData
 attenuationSettings=&/SA_MistRaven_GrimStrut_3P_Loop
 maxDistance=10300
 numChannels=2
 outputFormat=Stereo
 sampleRate=48000
 soundClassObject=/Game/Sounds/FortSoundClassesAndModes/Fort_06_Player_Attacks_Abilities
}
