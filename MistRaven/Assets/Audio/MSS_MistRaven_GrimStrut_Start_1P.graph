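# One-shot 5.1 MetaSound source for the MistRaven GrimStrut Start first-person sound, assigned to the melee weapon sound class.
# On play it picks one of three wave variations at random and plays it through a 5.1 wave player.
# Pitch shift is randomised within a base plus or minus spread range, which is zero by default.
# All six channels are scaled by a volume input, and the player's finish signals OnFinished.
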MetaSoundSource /MistRaven/Assets/Audio/MSS_MistRaven_GrimStrut_Start_1P
inputs [
 Random.Pitch.Base {type=Float}
 Random.Pitch.Spread {type=Float}
 Random.Volume.Volume Base {type=Float}
 Sounds {type=WaveAsset:Array}
 UE.Source.OnPlay {type=Trigger}
]
interfaces [UE.OutputFormat.5dot1:1.0, UE.Source.OneShot:1.0, UE.Source:1.1]
outputs [
 "UE.OutputFormat.5dot1.Audio:0" {type=Audio}
 "UE.OutputFormat.5dot1.Audio:1" {type=Audio}
 "UE.OutputFormat.5dot1.Audio:2" {type=Audio}
 "UE.OutputFormat.5dot1.Audio:3" {type=Audio}
 "UE.OutputFormat.5dot1.Audio:4" {type=Audio}
 "UE.OutputFormat.5dot1.Audio:5" {type=Audio}
 UE.Source.OneShot.OnFinished {type=Trigger}
]
pages [Default {
  input_defaults {
   Random.Pitch.Base=0
   Random.Pitch.Spread=0
   Random.Volume.Volume Base=1.2
   UE.Source.OnPlay=null
   Sounds [
    &/MistRaven_GrimStrut_Start_1P_01
    &/MistRaven_GrimStrut_Start_1P_02
    &/MistRaven_GrimStrut_Start_1P_03
   ]
  }
  nodes [
   ArrayRandomGet_0 = Array.Random Get:WaveAsset:Array(In Array=^Sounds, Next=^UE.Source.OnPlay) {also {defaults {Weights []}}}
   WavePlayer_0 = UE.Wave Player:5dot1(
    Play=@ArrayRandomGet_0.On Next
    Wave Asset=@ArrayRandomGet_0.Value
    Pitch Shift=UE.RandomFloat(
     Next=^UE.Source.OnPlay
     Max=UE.Add:Float(PrimaryOperand=^Random.Pitch.Base, AdditionalOperands=^Random.Pitch.Spread).Out
     Min=UE.Subtract:Float(PrimaryOperand=^Random.Pitch.Base, AdditionalOperands=^Random.Pitch.Spread).Out
    ).Value
   )
   Multiply_0 = UE.Multiply:Audio by Float(AdditionalOperands=^Random.Volume.Volume Base, PrimaryOperand=@WavePlayer_0.Out Low Frequency)
   Multiply_1 = UE.Multiply:Audio by Float(AdditionalOperands=^Random.Volume.Volume Base, PrimaryOperand=@WavePlayer_0.Out Front Center)
   Multiply_2 = UE.Multiply:Audio by Float(AdditionalOperands=^Random.Volume.Volume Base, PrimaryOperand=@WavePlayer_0.Out Side Right)
   Multiply_3 = UE.Multiply:Audio by Float(AdditionalOperands=^Random.Volume.Volume Base, PrimaryOperand=@WavePlayer_0.Out Side Left)
   Multiply_4 = UE.Multiply:Audio by Float(AdditionalOperands=^Random.Volume.Volume Base, PrimaryOperand=@WavePlayer_0.Out Front Left)
   Multiply_5 = UE.Multiply:Audio by Float(AdditionalOperands=^Random.Volume.Volume Base, PrimaryOperand=@WavePlayer_0.Out Front Right)
  ]
  links [
   {dst="@UE.OutputFormat.5dot1.Audio:3", src=Multiply_0.Out}
   {dst="@UE.OutputFormat.5dot1.Audio:2", src=Multiply_1.Out}
   {dst="@UE.OutputFormat.5dot1.Audio:5", src=Multiply_2.Out}
   {dst="@UE.OutputFormat.5dot1.Audio:4", src=Multiply_3.Out}
   {dst="@UE.OutputFormat.5dot1.Audio:0", src=Multiply_4.Out}
   {dst="@UE.OutputFormat.5dot1.Audio:1", src=Multiply_5.Out}
   {dst="@UE.Source.OneShot.OnFinished", src=WavePlayer_0.On Finished}
  ]
 }]
properties {
 assetImportData=&self:AssetImportData
 numChannels=6
 outputFormat=FiveDotOne
 sampleRate=48000
 soundClassObject=&Fort_WeaponMelee
}
