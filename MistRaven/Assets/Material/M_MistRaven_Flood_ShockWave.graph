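# Unlit, translucent, two-sided material for the MistRaven Flood shockwave effect, usable on Niagara mesh particles.
# Particle color RGB passes through MF_FX_EmissiveAdjustments to form the glow.
# A shockwave texture tiles twice across U and shifts along V by a dynamic "offset" parameter, letting the particle system drive its motion.
# Its red channel, faded by a powered V gradient and particle alpha, forms the opacity mask.
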Material /MistRaven/Assets/Material/M_MistRaven_Flood_ShockWave
graph {
 properties {
  bUsedWithNiagaraMeshParticles=true
  blendMode=BLEND_Translucent
  pixelDepthOffsetMode=PDOM_Legacy
  shadingModel=MSM_Unlit
  twoSided=true
  usedShadingModels=Unlit
 }
 nodes [
  n1 = ParticleColor()
  MaterialFunctionCall(Emissive=@n1.RGB) {materialFunction=&MF_FX_EmissiveAdjustments}
  n2 = Constant()
  n3 = DynamicParameter() {defaultValue=<rgba 0 1 1 1>, paramNames [offset,Param2,Param3,Param4]}
  n4 = Panner(Coordinate=Add(A=TextureCoordinate() {uTiling=2}, B=AppendVector(A=@n2, B=@n3.offset)))
  n5 = ComponentMask(None=TextureCoordinate()) {g=true}
  Multiply(A=@n1.A, B=Multiply(A=TextureSample(UVs=@n4) {texture=/MistRaven/Assets/Tex/T_MistRaven_Flood_ShockWave}.R, B=Power(Base=@n5)))
 ]
}
