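# Parent character material for MistRaven with a switchable, animated color wipe to a second diffuse texture.
# Pre-skinned local position, perturbed by two counter-panning cloud noise samples, forms a height gradient offset between TL_Min and TL_Max over time since TL_Time, scaled by TL_Duration.
# The gradient blends base color to DiffuseTwo and adds an emissive band at the front.
# UseWipeAnimation inserts this between the shared PreFX and PostFX material functions.
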Material /MistRaven/Assets/Materials/M_CharacterParent_2023_ColorWipe
graph {
 properties {
  bUseMaterialAttributes=true
  bUsedWithClothing=true
  bUsedWithMeshDeformer=true
  bUsedWithMorphTargets=true
  bUsedWithNanite=true
  bUsedWithNiagaraMeshParticles=true
  bUsedWithSkeletalMesh=true
  pixelDepthOffsetMode=PDOM_Legacy
  shadingModel=MSM_FromMaterialExpression
  usedShadingModels=Default Lit | Subsurface | Clear Coat | Subsurface Profile | Cloth
 }
 nodes [
  n1 = NamedRerouteUsage()
  n2 = ComponentMask(None=NamedRerouteUsage()) {b=true, r=true}
  n3 = Multiply(A=@n2) {constB=1.33}
  n4 = Add(
   A=TextureSample(UVs=Panner(Coordinate=@n2) {speedX=0.06, speedY=0.1}) {texture=&AtmosphericCloudNoise01}.R
   B=TextureSample(UVs=Panner(Coordinate=@n3) {speedX=-0.06, speedY=-0.1}) {texture=&AtmosphericCloudNoise01}.R
  )
  n5 = Add(
   A=Subtract(A=ComponentMask(None=@n1) {b=true}, B=Multiply(A=@n4) {constB=0.25})
   B=LinearInterpolate(
    A=ScalarParameter() {defaultValue=0.7, group=01_MistRaven_WipeAnimation, parameterName=TL_Min, sliderMax=1}
    B=ScalarParameter() {defaultValue=2.5, group=01_MistRaven_WipeAnimation, parameterName=TL_Max, sliderMax=1}
    Alpha=NamedRerouteUsage()
   ) {constA=-0.2, constB=1.2}
  )
  Ceil(None=@n5) {desc=Blend}
  Frac(None=Divide(A=Time()))
  n6 = MaterialFunctionCall() {materialFunction=&MF_CharacterParent_2023_PreFX}
  n7 = GetMaterialAttributes(None=@n6.Result) {attributeGetTypes [#1, #2]}
  n8 = TextureSampleParameter2D() {group=00_Texture Inputs, parameterName=DiffuseTwo, sortPriority=0, texture=/Game/Global/Textures/Default/Blanks/T_MidgreysRGB}
  n9 = LinearInterpolate(A=@n7.Base Color, B=@n8.RGB, Alpha=Saturate(None=Power(Base=@n5) {constExponent=32}))
  n10 = Add(A=@n5) {constB=-1}
  n11 = OneMinus(None=Saturate(None=Abs(None=@n10)))
  n12 = Add(
   A=@n7.Emissive Color
   B=Multiply(
    A=VectorParameter() {group=01_MistRaven_WipeAnimation, parameterName=EmissiveWipe_Color, defaultValue=<rgba 9.1855 0 30 1>}.RGB
    B=Power(Base=@n11) {constExponent=52}
   ) {desc=Emissiv}
  )
  MaterialFunctionCall(
   Material=StaticSwitchParameter(
    True=SetMaterialAttributes!(MaterialAttributes=@n7.MaterialAttributes, Base Color=@n9, Emissive Color=@n12) {attributeSetTypes [#1, #2]}
    False=@n6.Result
   ) {group=01_MistRaven_WipeAnimation, parameterName=UseWipeAnimation, sortPriority=0}
   M=@n6.M
   Pre Skinned Local Position=@n6.M
   Distance Blend=@n6.M
   Fresnel=@n6.M
  ) {materialFunction=&MF_CharacterParent_2023_PostFX}
  Multiply(A=@n4) {constB=0.5}
  n13 = LocalPosition() {localOrigin=InstancePreSkinning}
  NamedRerouteDeclaration(None=VertexInterpolator(VS=Divide(A=@n13.XYZ) {constB=100}).PS) {name=PSLP, nodeColor=<rgba 0 0.8594 1 1>}
  n14 = Subtract(A=Time(), B=ScalarParameter() {defaultValue=-99999, group=01_MistRaven_WipeAnimation, parameterName=TL_Time})
  n15 = Add(A=ScalarParameter() {defaultValue=1, group=01_MistRaven_WipeAnimation, parameterName=TL_Duration}) {constB=0}
  NamedRerouteDeclaration(None=Saturate(None=Divide(A=@n14, B=@n15))) {name=TL_AnimationCurve, nodeColor=<rgba 0.1484 1 0 1>}
  ScalarParameter() {group=01_MistRaven_WipeAnimation, parameterName=Time, sliderMax=1}
 ]
}
