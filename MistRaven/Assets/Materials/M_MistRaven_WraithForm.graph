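# Translucent unlit skeletal-mesh material for MistRaven's wraith form, built for a timed swirling smoke transformation.
# A timestamp-driven curve, with looping debug override, and pre-skinned height gradient shape a swirl offset that twists, bulges and pulls vertices toward a point.
# Opposed screen-space smoke panners add view-space wobble.
# Panning world-aligned noise and normals, lit by sky-light direction and masked by inverse Fresnel, form the smoke shading.
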Material /MistRaven/Assets/Materials/M_MistRaven_WraithForm
graph {
 properties {
  bAutomaticallySetUsageInEditor=false
  bUsedWithClothing=true
  bUsedWithMeshDeformer=true
  bUsedWithSkeletalMesh=true
  bUsedWithStaticLighting=true
  blendMode=BLEND_Translucent
  shadingModel=MSM_Unlit
  usedShadingModels=Unlit
 }
 nodes [
  n1 = Multiply(A=ScreenPosition().ViewportUV, B=ScalarParameter() {defaultValue=2, parameterName=SmokeWPO_Tiling, sliderMax=4})
  n2 = Panner(Coordinate=@n1) {speedX=-0.6, speedY=-0.5}
  n3 = Panner(Coordinate=@n1) {speedX=0.6, speedY=0.5}
  n4 = Add(
   A=Multiply(
    A=Constant3Vector() {constant=<rgba 1 0 0 0>}
    B=TextureSample(UVs=@n2) {texture=/Game/Global/Textures/Default/Tileable_Patterns/T_Blowing_Smoke_Pattern2_Tile_DISP}.R
   )
   B=Multiply(
    A=Constant3Vector() {constant=<rgba -1 0 0 0>}
    B=TextureSample(UVs=@n3) {texture=&T_Blowing_Smoke_Pattern2_Tile_DISP}.R
   )
  )
  n5 = ScalarParameter() {defaultValue=2, parameterName=WPO_Scale, sliderMax=10}
  Add(A=NamedRerouteUsage(), B=Transform(None=Multiply(A=@n4, B=@n5)) {transformSourceType=TRANSFORMSOURCE_View})
  CurveAtlasRowParameter(CurveTime=NamedRerouteUsage()) {atlas=&/CA_MistRaven, curve=&/C_FadeInOut, parameterName=CurveAtlas}
  n6 = Time()
  Divide(A=@n6) {constB=3}
  n7 = Subtract(A=Frac(None=Time() {period=5}), B=ScalarParameter() {parameterName=FlashStartTime})
  n8 = ScalarParameter() {defaultValue=1, group=Timestamp, parameterName=FlashDuration}
  MaterialFunctionCall!(In=Saturate(None=Divide(A=@n7, B=@n8))) {materialFunction=/Game/Effects/Fort_Effects/Materials/_Functions/MF_Smoothstep_EaseOut}
  n9 = Multiply(A=ScalarParameter() {defaultValue=20, parameterName=Scale}, B=VectorParameter() {parameterName=ScaleMult, defaultValue=<rgba 2 2 2 0>}.RGB)
  n10 = Multiply(A=Time(), B=ScalarParameter() {defaultValue=32, parameterName=PanSpeed})
  n11 = Multiply(A=@n10, B=Constant3Vector() {constant=<rgba 0.08 -0.08 -1 0>})
  n12 = Subtract(A=WorldPosition().XYZ, B=Multiply(A=ActorPositionWS()) {constB=0.95})
  n13 = ComponentMask(None=MaterialFunctionCall(
    TextureObject=TextureObject() {texture=&T_Smooth_tile}
    TextureSize=@n9
    WorldPosition=Add(A=@n11, B=@n12)
   ) {materialFunction=&WorldAlignedTexture}.XYZ Texture) {r=true}
  n14 = TextureObject() {samplerType=SAMPLERTYPE_Normal, texture=&T_Weap_Distort_N_02}
  n15 = Add(A=Multiply(A=Constant3Vector() {constant=<rgba -0.08 0.08 0.6 0>}, B=@n10), B=@n12)
  n16 = DotProduct(
   A=Multiply(A=SkyAtmosphereLightDirection()) {constB=-1}
   B=Transform(None=MaterialFunctionCall(TextureObject=@n14, TextureSize=@n9, WorldPosition=@n15) {materialFunction=&WorldAlignedNormal}.XYZ Texture)
  )
  Multiply(
   A=LinearInterpolate(A=Multiply(A=@n13) {constB=1.5}, Alpha=Saturate(None=@n16)) {constB=0.8}
   B=OneMinus(None=Fresnel() {baseReflectFraction=0, exponent=0.5})
  )
  n17 = Time()
  n18 = ScalarParameter() {defaultValue=-99999, parameterName=TL_Time}
  n19 = ScalarParameter() {defaultValue=0.8, parameterName=TL_Duration}
  n20 = Frac(None=@n6)
  n21 = LinearInterpolate(A=Divide(A=Subtract(A=@n17, B=@n18), B=Add(A=@n19) {constB=0}), B=@n20, Alpha=ScalarParameter() {parameterName=DebugAnimation, sliderMax=1})
  NamedRerouteDeclaration(None=OneMinus(None=Saturate(None=@n21))) {name=TL_AnimationCurve, nodeColor=<rgba 0.1484 1 0 1>}
  n22 = Subtract(
   A=TransformPosition(None=VectorParameter() {parameterName=PointLocation, defaultValue=<rgba 0 0 256 1>}.RGB) {transformType=TRANSFORMPOSSOURCE_World}
   B=WorldPosition().XYZ
  )
  n23 = LocalPosition() {localOrigin=InstancePreSkinning}
  n24 = ShaderStageSwitch(PixelShader=VertexInterpolator(VS=@n23.XYZ).PS, VertexShader=@n23.XYZ)
  n25 = PreSkinnedLocalBounds()
  n26 = NamedRerouteDeclaration(None=ComponentMask(None=Divide(A=@n24, B=@n25.Max)) {b=true}) {name=Z-Gradient, nodeColor=<rgba 0 1 0.2266 1>}
  n27 = NamedRerouteUsage()
  n28 = Saturate(None=Subtract(A=Add(A=@n26), B=Multiply(A=@n27) {constB=2}))
  n29 = Constant() {r=0.25}
  n30 = Multiply(
   A=VertexNormalWS()
   B=Multiply(A=ScalarParameter() {defaultValue=12, parameterName=WPO_Offset}, B=SphereMask(A=@n28, B=@n29) {attenuationRadius=0.25, hardnessPercent=0})
  )
  n31 = RotateAboutAxis(
   NormalizedRotationAxis=Transform(None=VectorParameter() {parameterName=RotationAxis, defaultValue=<rgba 0 0 1 1>}.RGB) {transformSourceType=TRANSFORMSOURCE_Local}
   RotationAngle=Multiply(A=ScalarParameter() {defaultValue=3, parameterName=TwistCount}, B=@n28)
   PivotPoint=Add(
    A=MaterialFunctionCall() {materialFunction=&ObjectPivotPoint}.Object Pivot Location
    B=VectorParameter() {parameterName=PivotOffset, defaultValue=<rgba 0 0 0 1>}.RGB
   )
   Position=WorldPosition().XYZ
  )
  NamedRerouteDeclaration(None=Add(A=LinearInterpolate(B=@n22, Alpha=@n28), B=Add(A=@n30, B=@n31))) {name=Swirl_WPO, nodeColor=<rgba 0 0.3437 1 1>}
  OneMinus(None=@n27)
  OneMinus(None=@n20)
  VectorParameter() {parameterName=Color, defaultValue=<rgba 0.0079 0.0082 0.01 0>}
  VertexColor()
 ]
}
