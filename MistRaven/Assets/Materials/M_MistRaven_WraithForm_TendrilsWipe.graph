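# Translucent unlit screen-space overlay material for MistRaven's Wraith Form, drawing tinted tendrils and a radial wipe across the view.
# Two curl-noise-deformed polar layers form the tendril mask, alongside scene-color distortion, edge desaturation and a simplified Low/Mobile path.
# The wipe erodes scrolling polar flame and cloud noise against an aspect-corrected, wobbling center gradient.
# Dynamic parameters drive fade and wipe progress; static switch ShowWipe selects wipe or tendril output.
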Material /MistRaven/Assets/Materials/M_MistRaven_WraithForm_TendrilsWipe
graph {
 properties {
  bAutomaticallySetUsageInEditor=false
  bUsedWithInstancedStaticMeshes=true
  bUsedWithNiagaraMeshParticles=true
  bUsedWithStaticLighting=true
  blendMode=BLEND_Translucent
  shadingModel=MSM_Unlit
  translucencyPass=MTP_BeforeDOF
  usedShadingModels=Unlit
 }
 nodes [
  n1 = ScreenPosition()
  n2 = MaterialFunctionCall(UVs=@n1.ViewportUV, Texture Scale=Constant() {r=1.5}) {materialFunction=&ScaleUVsByCenter}
  n3 = Multiply(A=Multiply(A=Add(A=@n2.UVs) {constB=-0.5}) {constB=2}, B=Power(Base=NamedRerouteUsage()) {constExponent=4})
  n4 = SceneColor!(UVs=Add(A=@n1.ViewportUV, B=Multiply(A=@n3) {constB=-0.1}))
  n5 = Desaturation(None=@n4.RGB)
  n6 = ScreenPosition()
  n7 = Abs(None=Multiply(A=Add(A=@n6.ViewportUV) {constB=-0.5}) {constB=2})
  n8 = MaterialFunctionCall(Float2=Power(Base=@n7, Exp=ScalarParameter() {defaultValue=2, parameterName=Desaturation_Exp, sliderMax=16})) {materialFunction=&BreakOutFloat2Components}
  n9 = Add(A=@n8.R, B=@n8.R)
  n10 = Saturate(None=Power(Base=Divide(A=@n9) {constB=1.5}) {constExponent=3})
  n11 = Constant3Vector()
  n12 = Add(A=Power(Base=NamedRerouteUsage())) {constB=-0.1}
  n13 = EyeAdaptationInverse(LightValueInput=LinearInterpolate(A=VectorParameter() {parameterName=Tendril_TintColor, defaultValue=<rgba 0.2258 0.0557 0.3906 1>}.RGB, B=@n11, Alpha=Saturate(None=@n12)))
  n14 = MaterialFunctionCall(Float2=Power(Base=@n7, Exp=ScalarParameter() {defaultValue=4, parameterName=TendrilFrame_Exp, sliderMax=16}) {constExponent=4}) {materialFunction=&BreakOutFloat2Components}
  n15 = Saturate(None=Divide(A=Add(A=@n14.R, B=@n14.R)))
  n16 = NamedRerouteUsage()
  n17 = EyeAdaptationInverse(LightValueInput=Multiply(A=Power(Base=@n15) {constExponent=0.4}, B=@n16), AlphaInput=Constant() {r=0.15})
  MaterialFunctionCall!(
   Default=LinearInterpolate(A=LinearInterpolate(A=@n4.RGB, B=@n5, Alpha=@n10) {constAlpha=1}, B=@n13.EyeAdaptationInverse, Alpha=@n17.EyeAdaptationInverse) {constAlpha=1}
   Low/Mobile=@n13.EyeAdaptationInverse
  ) {materialFunction=/Game/Effects/Fort_Effects/Materials/_Functions/MF_LowMobile_PathingSwitch}
  n18 = Multiply(A=DynamicParameter() {defaultValue=<rgba 0.5 0.5 1 0.5>, paramNames [FadeInOut,Wipe,Abort,NormTime]}.FadeInOut, B=NamedRerouteUsage())
  n19 = Multiply(A=@n18, B=@n17.EyeAdaptationInverse)
  MaterialFunctionCall(Default=@n18, Low/Mobile=@n19) {materialFunction=&MF_LowMobile_PathingSwitch}
  MaterialFunctionCall(Default=Multiply(A=@n5, B=Constant3Vector() {constant=<rgba 0.35 0.35 0.35 0>}), Low/Mobile=@n11) {materialFunction=&MF_LowMobile_PathingSwitch}
  n20 = DynamicParameter() {defaultValue=<rgba 0.5 0.5 1 0.5>, paramNames [FadeInOut,Wipe,Abort,NormTime]}
  Min(A=@n20.Wipe, B=@n20.Wipe)
  Multiply(A=@n16) {constB=1.5}
  n21 = MaterialFunctionCall(Float2=MaterialFunctionCall!(UV Coord=ScreenPosition().ViewportUV, LenghtScale=Constant() {r=3}, RadialScale=Constant() {r=0.25}) {materialFunction=/Game/Global/MaterialFunctions/TechArt/MF_PolarCoordinates}.Result) {materialFunction=&BreakOutFloat2Components}
  n22 = Multiply(A=@n21.G) {constB=0.25}
  n23 = Panner(Coordinate=MaterialFunctionCall(X=Add(A=@n21.R, B=@n22), Y=@n21.G) {materialFunction=&MakeFloat2}.Result) {speedX=-0.5, speedY=0.1}
  n24 = Constant()
  n25 = TextureSample!(UVs=@n23, Tex, "DDX(UVs)"=@n24, "DDY(UVs)"=@n24, Apply View MipBias) {
   mipValueMode=TMVM_Derivative
   samplerType=SAMPLERTYPE_LinearGrayscale
   texture=/Game/Weapons/FORT_Melee/Pickaxe_Cube_Ninja_Male/FX/T_FluidFlameTile
  }
  n26 = ScreenPosition()
  n27 = MaterialFunctionCall(Float2=@n26.ViewportUV) {materialFunction=&BreakOutFloat2Components}
  n28 = Multiply(A=Time(), B=ScalarParameter() {defaultValue=0.25, parameterName=Speed}) {constB=0.1}
  n29 = Sine(None=Add(A=@n27.G, B=@n28))
  n30 = Cosine(None=Add(A=@n27.R, B=@n28))
  n31 = Add(
   A=@n26.ViewportUV
   B=Multiply(
    A=MaterialFunctionCall(X=@n29, Y=@n30) {materialFunction=&MakeFloat2}.Result
    B=ScalarParameter() {defaultValue=0.03, parameterName=Intensity}
   ) {constB=0.1}
  )
  n32 = ViewSize()
  n33 = Divide(A=Multiply(A=Add(A=@n31) {constB=-0.5}, B=@n32), B=ComponentMask(None=@n32) {g=true})
  n34 = Add(A=Multiply(A=Add(A=@n33) {constB=0.5}) {constB=2}) {constB=-1}
  n35 = NamedRerouteDeclaration(None=Multiply(A=Length(None=@n34) {is=class MaterialExpressionLength}) {constB=0.5}) {name=ViewportCenterGradient, nodeColor=<rgba 0.9453 1 0 1>}
  n36 = Subtract(A=@n35, B=Add(A=Multiply(A=@n20.Wipe) {constB=1.8}) {constB=-0.5})
  n37 = Power(Base=Sine(None=Saturate(None=@n36))) {constExponent=1.5}
  n38 = Add(A=Saturate(None=Multiply(A=@n25.R) {constB=2}), B=Subtract(A=Multiply(A=@n37) {constB=1.5}))
  n39 = Multiply(A=Multiply(A=Saturate(None=@n38)) {constB=0.7}, B=@n35)
  n40 = Add(A=@n36) {constB=-0.08}
  n41 = Power(Base=Sine(None=Saturate(None=@n40))) {constExponent=3}
  n42 = Add(
   A=TextureSample(UVs=@n23, "DDX(UVs)"=@n24, "DDY(UVs)"=@n24) {mipValueMode=TMVM_Derivative, texture=/Game/TimeOfDay/LightFunction/T_CloudNoise_Artemis}.R
   B=Subtract(A=Multiply(A=@n41) {constB=1.9})
  )
  n43 = Multiply(A=Multiply(A=Saturate(None=@n42)) {constB=0.7, desc=0.7}, B=@n35)
  Multiply(A=@n39, B=@n43)
  n44 = Constant()
  n45 = Constant() {r=0.2}
  n46 = ScreenPosition()
  n47 = Multiply(A=ScreenPosition().ViewportUV) {constB=0.3}
  n48 = TextureSample!(UVs=Panner(Coordinate=@n47) {speedY=-0.05}, Tex, Apply View MipBias) {samplerType=SAMPLERTYPE_LinearColor, texture=/Game/Accessories/FORT_Backpacks/Backpack_M_MED_Carabus/FX/T_PerlinCurl}
  n49 = ScalarParameter() {defaultValue=0.15, parameterName=Deform_intensity, sliderMax=1}
  n50 = Multiply(A=Add(A=ComponentMask(None=@n48.RGB) {g=true, r=true}) {constB=-0.5}, B=@n49) {constB=0.06}
  n51 = Constant() {r=0.666}
  n52 = Divide(B=ScalarParameter() {defaultValue=4, parameterName=Tendril_Amount})
  n53 = MaterialFunctionCall(Float2=MaterialFunctionCall(UV Coord=Add(A=@n46.ViewportUV, B=@n50), LenghtScale=@n51, RadialScale=@n52) {materialFunction=&MF_PolarCoordinates}.Result) {materialFunction=&BreakOutFloat2Components}
  n54 = Frac(None=@n53.G)
  n55 = Multiply(A=@n53.R, B=Abs(None=Add(A=@n54) {constB=-0.5}))
  n56 = TextureSample(UVs=Panner(Coordinate=Multiply(A=@n47) {constB=0.666}) {speedY=0.05}) {samplerType=SAMPLERTYPE_LinearColor, texture=&T_PerlinCurl}
  n57 = Multiply(A=Add(A=ComponentMask(None=@n56.RGB) {g=true, r=true}) {constB=-0.5}, B=@n49) {constB=0.06}
  n58 = MaterialFunctionCall(Float2=MaterialFunctionCall(UV Coord=Add(A=@n46.ViewportUV, B=@n57), LenghtScale=@n51, RadialScale=@n52) {materialFunction=&MF_PolarCoordinates}.Result) {materialFunction=&BreakOutFloat2Components}
  n59 = Frac(None=Add(A=@n58.G) {constB=0.5})
  n60 = Multiply(A=@n58.R, B=Abs(None=Add(A=@n59) {constB=-0.5}))
  NamedRerouteDeclaration(None=Max(A=SmoothStep(Min=@n44, Max=@n45, Value=@n55) {constMax=0.15}, B=SmoothStep(Min=@n44, Max=@n45, Value=@n60) {constMax=0.15})) {name=TendrilsMask, nodeColor=<rgba 0 0.4141 1 1>}
  NamedRerouteDeclaration(None=SmoothStep(Value=@n36) {constMax=0.3, constMin=0.45}) {name=WipeMask, nodeColor=<rgba 0 0.2734 1 1>}
  ScreenPosition()
  n61 = Add(A=VectorParameter() {parameterName=Wipe_TintColor, defaultValue=<rgba 0.0608 0 0.1198 1>}.RGB, B=@n38)
  n62 = StaticBoolParameter() {parameterName=ShowWipe}
  StaticSwitch(
   True=LinearInterpolate(A=Multiply(A=@n61, B=@n39), B=Constant3Vector(), Alpha=@n43)
   False=Multiply(A=@n13.EyeAdaptationInverse, B=@n17.EyeAdaptationInverse)
   Value=@n62
  )
  StaticSwitch(True=@n43, False=@n19, Value=@n62)
  TextureCoordinate()
 ]
}
