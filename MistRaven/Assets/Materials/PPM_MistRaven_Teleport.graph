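# Post-process material for the Mist Raven teleport that lays animated, tinted smoke over a bulge-distorted screen.
# TL_Time and TL_Duration define normalized progress, intended to drive fade-in/out curve-atlas lookups for smoke and distortion.
# Smoke comes from cloud noise panned in polar coordinates, masked toward screen edges.
# Separate branches, with no visible link to the smoke blend, hold three-tone recoloring, depth-faded desaturation, a sine pulse and dual-panner noise.
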Material /MistRaven/Assets/Materials/PPM_MistRaven_Teleport
graph {
 properties {bAutomaticallySetUsageInEditor=false, materialDomain=MD_PostProcess, usedShadingModels=Default Lit}
 nodes [
  Frac(None=Time())
  n1 = Reroute()
  n2 = OneMinus(None=@n1)
  n3 = Saturate(None=ScalarParameter() {parameterName=ColorInvert_Blend, sliderMax=1})
  LinearInterpolate(
   A=ComponentMask(None=SceneTexture() {bFiltered=true, sceneTextureId=PPI_PostProcessInput0}.Color) {b=true, g=true, r=true}
   B=MaterialFunctionCall(
    A=VectorParameter() {parameterName=Color_Dark, sortPriority=10, defaultValue=<rgba 0 0 0 1>}.RGB
    B=VectorParameter() {parameterName=Color_Gray, sortPriority=11, defaultValue=<rgba 0.18 0.18 0.18 1>}.RGB
    C=VectorParameter() {parameterName=Color_Bright, sortPriority=12, defaultValue=<rgba 1 1 1 1>}.RGB
    Alpha=LinearInterpolate(A=@n1, B=@n2, Alpha=@n3)
   ) {materialFunction=&3ColorBlend}.Result
   Alpha=Saturate(None=ScalarParameter() {defaultValue=1, parameterName=Alpha, sliderMax=1})
  )
  n4 = NamedRerouteUsage()
  n5 = MaterialFunctionCall(Float2=MaterialFunctionCall!(UV Coord=ScreenPosition().ViewportUV, LenghtScale, RadialScale) {materialFunction=/Game/Global/MaterialFunctions/TechArt/MF_PolarCoordinates}.Result) {materialFunction=&BreakOutFloat2Components}
  n6 = MaterialFunctionCall(X=Power(Base=@n5.R) {constExponent=3}, Y=@n5.G) {materialFunction=&MakeFloat2}
  n7 = Constant()
  n8 = TextureSample!(UVs=Frac(None=Panner(Coordinate=@n6.Result) {speedX=-0.5}), Tex, "DDX(UVs)"=@n7, "DDY(UVs)"=@n7, Apply View MipBias) {mipValueMode=TMVM_Derivative, texture=&AtmosphericCloudNoise01}
  n9 = Power(Base=@n8.R) {constExponent=0.05}
  n10 = Multiply(A=Add(A=ScreenPosition().ViewportUV) {constB=-0.5}) {constB=2}
  n11 = MaterialFunctionCall(Float2=Power(Base=Abs(None=@n10)) {constExponent=5}) {materialFunction=&BreakOutFloat2Components}
  n12 = Saturate(None=Add(A=@n11.R, B=@n11.R))
  n13 = NamedRerouteUsage()
  LinearInterpolate(
   A=ComponentMask(None=SceneTexture(UVs=@n4) {bFiltered=true, sceneTextureId=PPI_PostProcessInput0}.Color) {b=true, g=true, r=true}
   B=MaterialFunctionCall(B=VectorParameter() {parameterName=Smoke_Color, defaultValue=<rgba 0.6979 0.2363 0.2363 1>}.RGB, Alpha=@n8.R) {materialFunction=&Lerp_3Color}.Result
   Alpha=LinearInterpolate(
    B=Multiply(A=@n9, B=@n12)
    Alpha=CurveAtlasRowParameter(CurveTime=@n13) {atlas=&/CA_MistRaven, curve=&/C_FadeInOut, parameterName=CurveAtlas}.R
   )
  )
  n14 = Desaturation(None=ComponentMask(None=SceneTexture() {bFiltered=true, sceneTextureId=PPI_PostProcessInput0}.Color) {b=true, g=true, r=true})
  n15 = SceneTexture() {bFiltered=true, sceneTextureId=PPI_SceneDepth}
  n16 = Divide(
   A=Add(
    A=ComponentMask(None=@n15.Color) {r=true}
    B=ScalarParameter() {defaultValue=-1000, parameterName=Fade_Offset, sliderMax=2000, sliderMin=-2000}
   )
   B=ScalarParameter() {defaultValue=5000, parameterName=Fade_Distance, sliderMax=10000, sliderMin=1000}
  ) {constB=1000}
  Min(A=@n14, B=Saturate(None=OneMinus(None=@n16)))
  Multiply(A=@n10, B=@n10)
  n17 = ScreenPosition()
  n18 = Add(A=@n17.ViewportUV) {constB=-0.5}
  n19 = SphereMask(A=@n17.ViewportUV, B=Constant() {r=0.5}) {attenuationRadius=0.5, hardnessPercent=0}
  n20 = ScalarParameter() {defaultValue=1, parameterName=Buldge_Power, sliderMax=8}
  n21 = NamedRerouteUsage()
  n22 = Multiply(
   A=@n18
   B=Multiply(
    A=Power(Base=@n19, Exp=@n20)
    B=CurveAtlasRowParameter(CurveTime=@n21) {atlas=&CA_MistRaven, curve=&C_FadeInOut, parameterName=CurveAtlas}.R
   ) {constB=0.1}
  )
  NamedRerouteDeclaration(None=Add(A=Add(A=@n18, B=@n22)) {constB=0.5}) {name=Deformed UV Coords, nodeColor=<rgba 0 1 0.6953 1>}
  n23 = Subtract(A=Time(), B=ScalarParameter() {defaultValue=-99999, parameterName=TL_Time})
  n24 = Add(A=ScalarParameter() {defaultValue=1, parameterName=TL_Duration}) {constB=0}
  NamedRerouteDeclaration(None=Saturate(None=Divide(A=@n23, B=@n24))) {name=TL_AnimationCurve, nodeColor=<rgba 0.1484 1 0 1>}
  OneMinus(None=@n14)
  n25 = Divide(A=Time(), B=Constant() {r=5})
  n26 = Constant() {r=0.5}
  n27 = Add(A=Multiply(A=Sine(None=@n25), B=@n26) {constB=0.5}, B=@n26)
  Power(Base=OneMinus(None=Power(Base=@n27) {constExponent=8})) {constExponent=8}
  n28 = ScreenPosition()
  n29 = Add(A=Multiply(A=Add(A=@n28.ViewportUV) {constB=-0.5}, B=ScalarParameter() {defaultValue=1, parameterName=Smoke_Tiling, sliderMax=4})) {constB=0.5}
  n30 = Panner(Coordinate=@n29) {speedX=0.3, speedY=0.5}
  n31 = Panner(Coordinate=@n29) {speedX=-0.4, speedY=-0.5}
  Saturate(None=Add(
    A=TextureSample!(UVs=@n30, Tex, Apply View MipBias) {texture=&AtmosphericCloudNoise01}.R
    B=TextureSample(UVs=@n31) {texture=&AtmosphericCloudNoise01}.R
   ))
  Saturate(None=ScalarParameter() {defaultValue=1, parameterName=Alpha, sliderMax=1})
  ScalarParameter() {parameterName=Buldge_Intensity, sliderMax=2, sliderMin=-2}
 ]
}
