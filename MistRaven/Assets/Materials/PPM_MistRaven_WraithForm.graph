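# Post-process material for MistRaven's Wraith Form screen effect, applied to scene color before bloom.
# It warps and desaturates the scene and adds purple-tinted tendrils built from two noise-distorted polar patterns toward the screen edges.
# A radial wipe with panning flame and cloud noise layers tinted and black bands over it, using EyeAdaptationInverse to compensate for exposure.
# Start-time and duration parameters define normalized start and end timelines, with curve-atlas channels shaping fade and wipe.
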Material /MistRaven/Assets/Materials/PPM_MistRaven_WraithForm
graph {
 properties {
  bAutomaticallySetUsageInEditor=false
  blendableLocation=BL_SceneColorBeforeBloom
  materialDomain=MD_PostProcess
  usedShadingModels=Default Lit
 }
 nodes [
  Constant() {r=0.5}
  Constant() {r=1}
  Constant()
  Constant() {r=0.5}
  Divide(B=EyeAdaptation().EyeAdaptation)
  Frac(None=Divide(A=Time()) {constB=5})
  n1 = Time()
  n2 = Subtract(A=@n1, B=ScalarParameter() {defaultValue=-99999, parameterName=TL_Time})
  LinearInterpolate(A=@n2, B=Frac(None=@n1), Alpha=ScalarParameter() {parameterName=DebugTime})
  n3 = ComponentMask(None=SceneTexture() {bFiltered=true, sceneTextureId=PPI_PostProcessInput0}.Color) {b=true, g=true, r=true}
  n4 = TextureCoordinate()
  n5 = Constant() {r=1.5}
  n6 = Multiply(A=Add(A=MaterialFunctionCall(UVs=@n4, Texture Scale=@n5) {materialFunction=&ScaleUVsByCenter}.UVs) {constB=-0.5}) {constB=2}
  n7 = Power(Base=NamedRerouteUsage()) {constExponent=4}
  n8 = Add(A=@n4, B=Multiply(A=Multiply(A=@n6, B=@n7)) {constB=-0.1})
  n9 = SceneTexture(UVs=@n8) {bFiltered=true, sceneTextureId=PPI_PostProcessInput0}
  n10 = Desaturation(None=ComponentMask(None=SceneTexture(UVs=@n8) {bFiltered=true, sceneTextureId=PPI_PostProcessInput0}.Color) {b=true, g=true, r=true})
  n11 = Multiply(A=Add(A=ScreenPosition().ViewportUV) {constB=-0.5}) {constB=2}
  n12 = Abs(None=@n11)
  n13 = MaterialFunctionCall(Float2=Power(Base=@n12, Exp=ScalarParameter() {defaultValue=2, parameterName=Desaturation_Exp, sliderMax=16})) {materialFunction=&BreakOutFloat2Components}
  n14 = Power(Base=Divide(A=Add(A=@n13.R, B=@n13.R)) {constB=1.5}) {constExponent=3}
  n15 = VectorParameter() {parameterName=Tendril_TintColor, defaultValue=<rgba 0.2258 0.0557 0.3906 1>}
  n16 = Multiply(A=@n10, B=Constant3Vector() {constant=<rgba 0.35 0.35 0.35 0>})
  n17 = NamedRerouteUsage()
  n18 = Saturate(None=Add(A=Power(Base=@n17)) {constB=-0.1})
  n19 = MaterialFunctionCall(Float2=Power(Base=@n12, Exp=ScalarParameter() {defaultValue=4, parameterName=TendrilFrame_Exp, sliderMax=16}) {constExponent=4}) {materialFunction=&BreakOutFloat2Components}
  n20 = Add(A=@n19.R, B=@n19.R)
  n21 = Power(Base=Saturate(None=Divide(A=@n20))) {constExponent=0.4}
  n22 = Multiply(A=NamedRerouteUsage()) {constB=1.5}
  n23 = LinearInterpolate(
   A=LinearInterpolate(A=ComponentMask(None=@n9.Color) {b=true, g=true, r=true}, B=@n10, Alpha=Saturate(None=@n14)) {constAlpha=1}
   B=EyeAdaptationInverse(LightValueInput=LinearInterpolate(A=@n15.RGB, B=@n16, Alpha=@n18)).EyeAdaptationInverse
   Alpha=EyeAdaptationInverse(LightValueInput=Multiply(A=@n21, B=@n22), AlphaInput=Constant() {r=0.15}).EyeAdaptationInverse
  ) {constAlpha=1}
  n24 = Multiply(
   A=CurveAtlasRowParameter(CurveTime=NamedRerouteUsage()) {atlas=&/CA_MistRaven, curve=&/C_FadeInOut, parameterName=CurveAtlas}.R
   B=NamedRerouteUsage()
  )
  n25 = MaterialFunctionCall(Float2=MaterialFunctionCall!(UV Coord, LenghtScale=Constant() {r=3}, RadialScale=Constant() {r=0.25}) {materialFunction=/Game/Global/MaterialFunctions/TechArt/MF_PolarCoordinates}.Result) {materialFunction=&BreakOutFloat2Components}
  n26 = Multiply(A=@n25.G) {constB=0.25}
  n27 = Panner(Coordinate=MaterialFunctionCall(X=Add(A=@n25.R, B=@n26), Y=@n25.G) {materialFunction=&MakeFloat2}.Result) {speedX=-0.5, speedY=0.1}
  n28 = Constant()
  n29 = TextureSample!(UVs=@n27, Tex, "DDX(UVs)"=@n28, "DDY(UVs)"=@n28, Apply View MipBias) {
   mipValueMode=TMVM_Derivative
   samplerType=SAMPLERTYPE_LinearGrayscale
   texture=/Game/Weapons/FORT_Melee/Pickaxe_Cube_Ninja_Male/FX/T_FluidFlameTile
  }
  n30 = ScreenPosition()
  n31 = MaterialFunctionCall(Float2=@n30.ViewportUV) {materialFunction=&BreakOutFloat2Components}
  n32 = Multiply(A=Time(), B=ScalarParameter() {defaultValue=0.25, parameterName=Speed}) {constB=0.1}
  n33 = Sine(None=Add(A=@n31.G, B=@n32))
  n34 = Cosine(None=Add(A=@n31.R, B=@n32))
  n35 = Add(
   A=@n30.ViewportUV
   B=Multiply(
    A=MaterialFunctionCall(X=@n33, Y=@n34) {materialFunction=&MakeFloat2}.Result
    B=ScalarParameter() {defaultValue=0.03, parameterName=Intensity}
   ) {constB=0.1}
  )
  n36 = ViewSize()
  n37 = Divide(A=Multiply(A=Add(A=@n35) {constB=-0.5}, B=@n36), B=ComponentMask(None=@n36) {g=true})
  n38 = Add(A=Multiply(A=Add(A=@n37) {constB=0.5}) {constB=2}) {constB=-1}
  n39 = NamedRerouteDeclaration(None=Multiply(A=Length(None=@n38) {is=class MaterialExpressionLength}) {constB=0.5}) {name=ViewportCenterGradient, nodeColor=<rgba 0.9453 1 0 1>}
  n40 = Min(
   A=CurveAtlasRowParameter(CurveTime=NamedRerouteUsage()) {atlas=&CA_MistRaven, curve=&C_FadeInOut, parameterName=CurveAtlas}.G
   B=CurveAtlasRowParameter(CurveTime=NamedRerouteUsage()) {atlas=&CA_MistRaven, curve=&C_FadeInOut, parameterName=CurveAtlas}.B
  )
  n41 = Subtract(A=@n39, B=Add(A=Multiply(A=@n40) {constB=1.8}) {constB=-0.5})
  n42 = Power(Base=Sine(None=Saturate(None=@n41))) {constExponent=1.5}
  n43 = Add(A=Saturate(None=Multiply(A=@n29.R) {constB=2}), B=Subtract(A=Multiply(A=@n42) {constB=1.5}))
  n44 = Add(A=VectorParameter() {parameterName=Wipe_TintColor, defaultValue=<rgba 0.0608 0 0.1198 1>}.RGB, B=@n43)
  n45 = Multiply(A=Saturate(None=@n43)) {constB=0.7}
  n46 = Saturate(None=Add(A=@n41) {constB=-0.08})
  n47 = Multiply(A=Power(Base=Sine(None=@n46)) {constExponent=3}) {constB=1.9}
  n48 = Saturate(None=Add(
    A=TextureSample(UVs=@n27, "DDX(UVs)"=@n28, "DDY(UVs)"=@n28) {mipValueMode=TMVM_Derivative, texture=/Game/TimeOfDay/LightFunction/T_CloudNoise_Artemis}.R
    B=Subtract(A=@n47)
   ))
  LinearInterpolate(
   A=LinearInterpolate(
    A=LinearInterpolate(A=@n3, B=@n23, Alpha=@n24) {constAlpha=1}
    B=EyeAdaptationInverse(LightValueInput=@n44).EyeAdaptationInverse
    Alpha=Multiply(A=@n45, B=@n39)
   )
   B=Constant3Vector()
   Alpha=Multiply(A=Multiply(A=@n48) {constB=0.7}, B=@n39)
  )
  Multiply(
   A=ComponentMask(None=SceneTexture() {bFiltered=true, sceneTextureId=PPI_PostProcessInput0}.Color) {b=true, g=true, r=true}
   B=Constant3Vector() {constant=<rgba 1 0 0 0>}
  )
  n49 = Constant()
  n50 = Constant() {r=0.2}
  n51 = TextureCoordinate()
  n52 = Multiply(A=TextureCoordinate()) {constB=0.3}
  n53 = TextureSample!(UVs=Panner(Coordinate=@n52) {speedY=-0.05}, Tex, Apply View MipBias) {samplerType=SAMPLERTYPE_LinearColor, texture=/Game/Accessories/FORT_Backpacks/Backpack_M_MED_Carabus/FX/T_PerlinCurl}
  n54 = ScalarParameter() {defaultValue=0.15, parameterName=Deform_intensity, sliderMax=1}
  n55 = Multiply(A=Add(A=ComponentMask(None=@n53.RGB) {g=true, r=true}) {constB=-0.5}, B=@n54) {constB=0.06}
  n56 = Constant() {r=0.666}
  n57 = Divide(B=ScalarParameter() {defaultValue=4, parameterName=Tendril_Amount})
  n58 = MaterialFunctionCall(Float2=MaterialFunctionCall(UV Coord=Add(A=@n51, B=@n55), LenghtScale=@n56, RadialScale=@n57) {materialFunction=&MF_PolarCoordinates}.Result) {materialFunction=&BreakOutFloat2Components}
  n59 = Frac(None=@n58.G)
  n60 = Multiply(A=@n58.R, B=Abs(None=Add(A=@n59) {constB=-0.5}))
  n61 = TextureSample(UVs=Panner(Coordinate=Multiply(A=@n52) {constB=0.666}) {speedY=0.05}) {samplerType=SAMPLERTYPE_LinearColor, texture=&T_PerlinCurl}
  n62 = Multiply(A=Add(A=ComponentMask(None=@n61.RGB) {g=true, r=true}) {constB=-0.5}, B=@n54) {constB=0.06}
  n63 = MaterialFunctionCall(Float2=MaterialFunctionCall(UV Coord=Add(A=@n51, B=@n62), LenghtScale=@n56, RadialScale=@n57) {materialFunction=&MF_PolarCoordinates}.Result) {materialFunction=&BreakOutFloat2Components}
  n64 = Frac(None=Add(A=@n63.G) {constB=0.5})
  n65 = Multiply(A=@n63.R, B=Abs(None=Add(A=@n64) {constB=-0.5}))
  NamedRerouteDeclaration(None=Max(A=SmoothStep(Min=@n49, Max=@n50, Value=@n60) {constMax=0.15}, B=SmoothStep(Min=@n49, Max=@n50, Value=@n65) {constMax=0.15})) {name=TendrilsMask, nodeColor=<rgba 0 0.4141 1 1>}
  n66 = Add(A=ScalarParameter() {defaultValue=5.75, parameterName=TL_Duration}) {constB=0}
  NamedRerouteDeclaration(None=Saturate(None=Divide(A=@n2, B=@n66))) {name=TL_AnimationCurve, nodeColor=<rgba 0.1484 1 0 1>}
  n67 = Subtract(A=Time(), B=ScalarParameter() {defaultValue=99999, parameterName=TL_End_Time})
  n68 = Add(A=ScalarParameter() {defaultValue=1, parameterName=TL_End_Duration}) {constB=0}
  NamedRerouteDeclaration(None=Saturate(None=Divide(A=@n67, B=@n68))) {name=TL_End_AnimationCurve, nodeColor=<rgba 0.1484 1 0 1>}
  NamedRerouteDeclaration(None=SmoothStep(Value=@n41) {constMax=0.3, constMin=0.45}) {name=WipeMask, nodeColor=<rgba 0 0.2734 1 1>}
  NamedRerouteUsage()
  OneMinus(None=@n10)
  n69 = MaterialFunctionCall() {materialFunction=&DebugTimeSine}
  n70 = Sine(None=@n69.Result)
  n71 = Cosine(None=@n69.Result)
  Saturate(None=DotProduct(A=@n11, B=MaterialFunctionCall(X=@n70, Y=@n71) {materialFunction=&MakeFloat2}.Result))
  n72 = TextureCoordinate()
  n73 = Add(A=Power(Base=ComponentMask(None=@n72) {r=true}) {constExponent=1}) {constB=-0.75}
  n74 = Add(A=Multiply(A=Sine(None=@n73)) {constB=0.5}) {constB=0.5}
  Saturate(None=OneMinus(None=Power(Base=@n74) {constExponent=1}))
  ScalarParameter() {defaultValue=0.5, sliderMax=1}
  TextureCoordinate()
 ]
}
