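# Physics asset for the dynamic hood bones of the MistRaven WraithForm mesh, giving the hood limited physical swing.
# A small capsule on backpack_root anchors a chain of two identical longer capsules on dyn_hood_01 and dyn_hood_02.
# Two constraints link the chain, each with matching limits that allow moderate, asymmetric swing and very little twist.
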PhysicsAsset /MistRaven/Assets/Mesh/WraithForm_MistRaven_Dyn_hood
bodies {
 backpack_root [{shapeType=Capsule, length=1, radius=1, shapeName=backpack_root_capsule, rotation=<rot 90 0 0>}]
 dyn_hood_01 [{shapeType=Capsule, length=19.0938, radius=2, shapeName=dyn_hood_01_capsule, center=<4 0 0>}]
 dyn_hood_02 [{shapeType=Capsule, length=19.0938, radius=2, shapeName=dyn_hood_01_capsule, center=<4 0 0>}]
}
constraints [
 {bone1Name=dyn_hood_02, bone2Name=dyn_hood_01, swing1LimitDegrees=25, swing2LimitDegrees=15, twistLimitDegrees=2}
 {bone1Name=dyn_hood_01, bone2Name=backpack_root, swing1LimitDegrees=25, swing2LimitDegrees=15, twistLimitDegrees=2}
]
