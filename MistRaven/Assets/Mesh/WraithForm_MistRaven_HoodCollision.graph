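# Physics asset giving collision shapes to the hood bones of the MistRaven WraithForm mesh.
# The dyn_hood_01 body uses one capsule, while dyn_hood_02 uses two narrower capsules offset along the bone's X axis, all with their long axes crosswise to it.
# A single constraint joins the two bodies with both swings and twist set Free, so its angular limit angles do not restrict rotation.
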PhysicsAsset /MistRaven/Assets/Mesh/WraithForm_MistRaven_HoodCollision
bodies {
 dyn_hood_01 [{shapeType=Capsule, length=19.8438, radius=5.5312, shapeName=dyn_hood_01_capsule, center=<5.3899 -3.487 0>}]
 dyn_hood_02 [
  {shapeType=Capsule, length=12, radius=4.125, shapeName=dyn_hood_02_capsule, center=<6.2238 -1.3315 0>}
  {shapeType=Capsule, length=15.7812, radius=4.125, shapeName=dyn_hood_02_capsule, center=<1.4044 -1.7746 0>}
 ]
}
constraints [{
  bone1Name=dyn_hood_02
  bone2Name=dyn_hood_01
  swing1LimitDegrees=45
  swing1Motion=Free
  swing2LimitDegrees=45
  swing2Motion=Free
  twistLimitDegrees=45
  twistMotion=Free
 }]
